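# Control Rig for the WoolExpo female medium body that procedurally corrects thigh strap and pant cuff bones as the legs move.
# Spherical pose readers on thighs and feet write alpha variables.
# Parent constraints blend front straps and cuffs toward body bones, then AlphaInterp-smoothed alphas weight corrective Modify Transforms, alongside a constant rear-strap offset.
# Calf bend alphas are remapped from knee rotation and stored, but never read here.
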ControlRigBlueprint /BRCosmetics/Characters/Player/Female/Medium/Bodies/F_MED_WoolExpo/Meshes/F_MED_WoolExpo_CtrlRig
graphs [
 RigVMFunctionLibrary {links [], nodes []}
 RigVMFunction_Sequence_1_ContainedGraph {nodes [
   n1 = RigVMFunctionEntryNode()
   RigVMFunction_Sequence = &RigVMFunction_Sequence(ExecuteContext=@n1.ExecuteContext)
   n2 = &RigVMFunction_Sequence(ExecuteContext=@RigVMFunction_Sequence.B)
   RigVMFunctionReturnNode(A=@RigVMFunction_Sequence.A, B=@n2.A, C=@n2.B)
  ]}
 RigVMModel {nodes [
   RigUnit_BeginExecution = &RigUnit_BeginExecution()
   RigVMFunction_Sequence = RigVMAggregateNode(ExecuteContext=@RigUnit_BeginExecution.ExecutePin)
   n1 = &RigUnit_SphericalPoseReader(
    ActiveRegionScaleFactors=(PositiveWidth=4,NegativeWidth=3,PositiveHeight=2.5,NegativeHeight=2)
    ActiveRegionSize=0.1
    Debug=(bDrawDebug=false,bDraw2D=False,bDrawLocalAxes=False,DebugScale=25,DebugSegments=20,DebugThickness=0.25)
    DriverAxis=(X=-1,Y=0,Z=0)
    DriverItem=(Type=Bone,Name="thigh_l")
    FalloffRegionScaleFactors=(PositiveWidth=1,NegativeWidth=3,PositiveHeight=1,NegativeHeight=0.5)
    FalloffSize=0.4
    FlipHeightScaling=False
    FlipWidthScaling=False
    OptionalParentItem=(Type=Bone,Name="None")
    RotationOffset=(X=100,Y=0,Z=-20)
    ExecutePin=@RigVMFunction_Sequence.A
   )
   n2 = RigVMVariableNode(Value=@n1.OutputParam, ExecuteContext=@RigVMFunction_Sequence.B) {variable=L thighForward Alpha}
   n3 = &RigUnit_SphericalPoseReader(
    ActiveRegionScaleFactors=(PositiveWidth=4,NegativeWidth=3,PositiveHeight=2.5,NegativeHeight=2)
    ActiveRegionSize=0.1
    Debug=(bDrawDebug=false,bDraw2D=False,bDrawLocalAxes=False,DebugScale=25,DebugSegments=20,DebugThickness=0.25)
    DriverAxis=(X=1,Y=0,Z=0)
    DriverItem=(Type=Bone,Name="thigh_r")
    FalloffRegionScaleFactors=(PositiveWidth=1,NegativeWidth=3.6,PositiveHeight=1,NegativeHeight=0.5)
    FalloffSize=0.4
    FlipHeightScaling=False
    FlipWidthScaling=true
    OptionalParentItem=(Type=Bone,Name="None")
    RotationOffset=(X=-80,Y=0,Z=-35.3405)
    ExecutePin=@n1.ExecutePin
   )
   n4 = RigVMVariableNode(Value=@n3.OutputParam, ExecuteContext=@n2.ExecuteContext) {variable=R thighForward Alpha}
   n5 = &RigUnit_SphericalPoseReader(
    ActiveRegionScaleFactors=(PositiveWidth=1,NegativeWidth=1,PositiveHeight=1,NegativeHeight=1)
    ActiveRegionSize=0.1
    Debug=(bDrawDebug=true,bDraw2D=False,bDrawLocalAxes=False,DebugScale=25,DebugSegments=20,DebugThickness=0.25)
    DriverAxis=(X=0,Y=1,Z=0)
    DriverItem=(Type=Bone,Name="foot_l")
    FalloffRegionScaleFactors=(PositiveWidth=1,NegativeWidth=1,PositiveHeight=1,NegativeHeight=1)
    FalloffSize=0.15
    FlipHeightScaling=False
    FlipWidthScaling=False
    OptionalParentItem=(Type=Bone,Name="None")
    RotationOffset=(X=98.6371,Y=0,Z=74.0326)
    ExecutePin=@n3.ExecutePin
   )
   n6 = RigVMVariableNode(Value=@n5.OutputParam, ExecuteContext=@n4.ExecuteContext) {variable=L FootBend Alpha}
   n7 = &RigUnit_SphericalPoseReader(
    ActiveRegionScaleFactors=(PositiveWidth=1,NegativeWidth=1,PositiveHeight=1,NegativeHeight=1)
    ActiveRegionSize=0.1
    Debug=(bDrawDebug=true,bDraw2D=False,bDrawLocalAxes=False,DebugScale=25,DebugSegments=20,DebugThickness=0.25)
    DriverAxis=(X=0,Y=-1,Z=0)
    DriverItem=(Type=Bone,Name="foot_r")
    FalloffRegionScaleFactors=(PositiveWidth=1,NegativeWidth=1,PositiveHeight=1,NegativeHeight=1)
    FalloffSize=0.175
    FlipHeightScaling=False
    FlipWidthScaling=False
    OptionalParentItem=(Type=Bone,Name="None")
    RotationOffset=(X=-90,Y=0,Z=62.7858)
    ExecutePin=@n5.ExecutePin
   )
   RigVMVariableNode(Value=@n7.OutputParam, ExecuteContext=@n6.ExecuteContext) {variable=R FootBend Alpha}
   &RigUnit_ParentConstraint(
    AdvancedSettings=(InterpolationType=Average,RotationOrderForFilter=XZY)
    Child=(Type=Bone,Name="dyn_strap_fr_l_base")
    Parents=((Item=(Type=Bone,Name="thigh_l"),Weight=1),(Item=(Type=Bone,Name="spine_02"),Weight=0.018),(Item=(Type=Bone,Name="pelvis"),Weight=0))
    Weight=1
    ExecutePin=@RigVMFunction_Sequence.C
   )
   > &RigUnit_ParentConstraint(
    AdvancedSettings=(InterpolationType=Average,RotationOrderForFilter=XZY)
    Child=(Type=Bone,Name="dyn_strap_fr_r_base")
    Parents=((Item=(Type=Bone,Name="thigh_r"),Weight=0.4368),(Item=(Type=Bone,Name="thigh_r"),Weight=1))
    Weight=1
   )
   > &RigUnit_ModifyTransforms(
    ItemToModify="((Item=(Type=Bone,Name=\"dyn_strap_bck_r_base\"),Transform=(Rotation=(X=0,Y=0.0872,Z=0,W=-0.9962),Translation=(X=0,Y=0,Z=0),Scale3D=(X=1,Y=1,Z=1))),(Item=(Type=Bone,Name=\"dyn_strap_bck_l_base\"),Transform=(Rotation=(X=0,Y=-0.0872,Z=0,W=0.9962),Translation=(X= ...(truncated)"
    Weight=1
    WeightMinimum=0
   )
   > &RigUnit_ParentConstraint(
    AdvancedSettings=(InterpolationType=Average,RotationOrderForFilter=XZY)
    Child=(Type=Bone,Name="dyn_pants_l_bck_02")
    Parents=((Item=(Type=Bone,Name="foot_l"),Weight=1))
    Weight=0.7
   )
   > &RigUnit_ParentConstraint(
    AdvancedSettings=(InterpolationType=Average,RotationOrderForFilter=XZY)
    Child=(Type=Bone,Name="dyn_pants_r_bck_02")
    Parents=((Item=(Type=Bone,Name="foot_r"),Weight=1))
    Weight=0.5
   )
   > &RigUnit_ParentConstraint(
    AdvancedSettings=(InterpolationType=Average,RotationOrderForFilter=XZY)
    Child=(Type=Bone,Name="dyn_pants_l_outer_02")
    Parents=((Item=(Type=Bone,Name="foot_l"),Weight=1))
    Weight=0.5
   )
   > &RigUnit_ParentConstraint(
    AdvancedSettings=(InterpolationType=Average,RotationOrderForFilter=XZY)
    Child=(Type=Bone,Name="dyn_pants_r_outer_02")
    Parents=((Item=(Type=Bone,Name="foot_r"),Weight=1))
    Weight=0.5
   )
   > &RigUnit_ParentConstraint(
    AdvancedSettings=(InterpolationType=Average,RotationOrderForFilter=XZY)
    Child=(Type=Bone,Name="dyn_pants_l_fr_02")
    Parents=((Item=(Type=Bone,Name="foot_l"),Weight=1))
    Weight=0.5
   )
   > &RigUnit_ParentConstraint(
    AdvancedSettings=(InterpolationType=Average,RotationOrderForFilter=XZY)
    Child=(Type=Bone,Name="dyn_pants_r_fr_02")
    Parents=((Item=(Type=Bone,Name="foot_r"),Weight=1))
    Weight=0.7
   )
   > &RigUnit_ParentConstraint(
    AdvancedSettings=(InterpolationType=Average,RotationOrderForFilter=XZY)
    Child=(Type=Bone,Name="dyn_pants_l_inner_02")
    Parents=((Item=(Type=Bone,Name="foot_l"),Weight=1))
    Weight=0.5
   )
   > &RigUnit_ParentConstraint(
    AdvancedSettings=(InterpolationType=Average,RotationOrderForFilter=XZY)
    Child=(Type=Bone,Name="dyn_pants_r_inner_02")
    Parents=((Item=(Type=Bone,Name="foot_r"),Weight=1))
    Weight=0.5
   )
   n8 = &RigVMFunction_AlphaInterp(
    InterpSpeedDecreasing=30
    InterpSpeedIncreasing=30
    Scale=1
    bClampResult=true
    Value=RigVMVariableNode() {variable=L thighForward Alpha}.Value
   )
   > &RigUnit_ModifyTransforms(
    ItemToModify="((Item=(Type=Bone,Name=\"dyn_strap_fr_l_base\"),Transform=(Rotation=(X=0.0226,Y=0.2578,Z=-0.25,W=-0.933),Translation=(X=3.8948,Y=-0.7367,Z=-0.112),Scale3D=(X=1,Y=1,Z=1))),(Item=(Type=Bone,Name=\"dyn_strap_bck_l_base\"),Transform=(Rotation=(X=0,Y=0.0872,Z=0,W=- ...(truncated)"
    WeightMinimum=0
    Weight=@n8.Result
   )
   n9 = &RigVMFunction_AlphaInterp(
    InterpSpeedDecreasing=30
    InterpSpeedIncreasing=30
    Scale=1
    bClampResult=true
    Value=RigVMVariableNode() {variable=R thighForward Alpha}.Value
   )
   > &RigUnit_ModifyTransforms(
    ItemToModify=((Item=(Type=Bone,Name="dyn_strap_fr_r_base"),Transform=(Rotation=(X=0.0151,Y=0.173,Z=0.0858,W=-0.9811),Translation=(X=0.1363,Y=0.7727,Z=0),Scale3D=(X=1,Y=1,Z=1))))
    WeightMinimum=0
    Weight=@n9.Result
   )
   n10 = &RigVMFunction_AlphaInterp(InterpSpeedDecreasing=25, InterpSpeedIncreasing=25, Scale=1, bClampResult=true, Value=RigVMVariableNode() {variable=L FootBend Alpha}.Value)
   > &RigUnit_ModifyTransforms(
    ItemToModify="((Item=(Type=Bone,Name=\"dyn_pants_l_bck_02\"),Transform=(Rotation=(X=-0.0594,Y=0.171,Z=0.1632,W=-0.9698),Translation=(X=-0.0044,Y=0,Z=0.0077),Scale3D=(X=1,Y=1,Z=1))),(Item=(Type=Bone,Name=\"dyn_pants_l_inner_02\"),Transform=(Rotation=(X=0,Y=0,Z=0,W=-1),Transl ...(truncated)"
    WeightMinimum=0
    Weight=@n10.Result
   )
   n11 = &RigVMFunction_AlphaInterp(InterpSpeedDecreasing=25, InterpSpeedIncreasing=25, Scale=1, bClampResult=true, Value=RigVMVariableNode() {variable=R FootBend Alpha}.Value)
   > &RigUnit_ModifyTransforms(
    ItemToModify=((Item=(Type=Bone,Name="dyn_pants_r_bck_02"),Transform=(Rotation=(X=0.0151,Y=0.173,Z=-0.0858,W=-0.9811),Translation=(X=0,Y=0,Z=0),Scale3D=(X=0.8,Y=1,Z=1))))
    WeightMinimum=0
    Weight=@n11.Result
   )
   n12 = &RigUnit_GetRelativeTransformForItem(Child=(Type=Bone,Name="calf_r"), Parent=(Type=Bone,Name="thigh_r"), bChildInitial=False, bParentInitial=False)
   RigVMVariableNode(
    Value=&RigVMFunction_MathDoubleRemap(
     SourceMaximum=100
     SourceMinimum=45
     TargetMaximum=1
     TargetMinimum=0
     Value=&RigVMFunction_MathQuaternionToEuler(RotationOrder=ZYX, Value=@n12.RelativeTransform.Rotation).Result.Z
    ).Result
    ExecuteContext=@n14.ExecuteContext
   ) {variable=R calfBend Alpha}
   n13 = &RigUnit_GetRelativeTransformForItem(Child=(Type=Bone,Name="calf_l"), Parent=(Type=Bone,Name="thigh_l"), bChildInitial=False, bParentInitial=False)
   RigVMFunction_MathQuaternionToEuler = &RigVMFunction_MathQuaternionToEuler(RotationOrder=ZYX, Value=@n13.RelativeTransform.Rotation)
   Remap = &RigVMFunction_MathDoubleRemap(SourceMaximum=100, SourceMinimum=45, TargetMaximum=1, TargetMinimum=0, Value=@RigVMFunction_MathQuaternionToEuler.Result.Z)
   n14 = RigVMVariableNode(Value=@Remap.Result, ExecuteContext=@n7.ExecutePin) {variable=L calfBend Alpha}
  ]}
]
hierarchy [
 root {loc=<0 0 0>, rot=<0 0 0 1>}
 pelvis {parent=root, loc=<0 -0.5301 98.0571>, rot=<0.0127 0.707 -0.0127 -0.707>}
 spine_01 {parent=pelvis, loc=<6.1208 -0.9328 0>, rot=<0 0 0.034 -0.9994>}
 spine_02 {parent=spine_01, loc=<6.5997 -0.2095 0>, rot=<0 0 0.0238 -0.9997>}
 spine_03 {parent=spine_02, loc=<6.4914 -0.1237 0>, rot=<0 0 0.0149 -0.9999>}
 spine_04 {parent=spine_03, loc=<6.9598 0.0794 0>, rot=<0 0 -0.0132 -0.9999>}
 spine_05 {parent=spine_04, loc=<6.9987 0 0>, rot=<0 0 -0.0015 -1>}
 neck_01 {parent=spine_05, loc=<11.3138 0.1529 0>, rot=<0 0 -0.2003 -0.9797>}
 neck_02 {parent=neck_01, loc=<4.5627 0 0>, rot=<0 0 0.0504 -0.9987>}
 head {parent=neck_02, loc=<5.3914 0 0>, rot=<0 0 0.115 -0.9934>}
 clavicle_l {parent=spine_05, loc=<5.4831 -0.1053 -3.2946>, rot=<0.0465 -0.7289 -0.009 -0.6829>}
 upperarm_l {parent=clavicle_l, loc=<9.8658 -1.2136 -0.4105>, rot=<0.0164 -0.3245 -0.0269 -0.9454>}
 lowerarm_l {parent=upperarm_l, loc=<26.465 -0.8345 0.259>, rot=<0.0036 0.0161 -0.2298 -0.9731>}
 hand_l {parent=lowerarm_l, loc=<22.7355 -0.3203 -0.6105>, rot=<-0.6784 -0.0223 0.0735 -0.7307>}
 thumb_01_l {parent=hand_l, loc=<1.3118 -0.3191 -2.459>, rot=<0.6546 -0.4042 0.0664 -0.6354>}
 thumb_02_l {parent=thumb_01_l, loc=<5.9336 -0.0448 0.0401>, rot=<-0.0037 0.0213 0.298 -0.9543>}
 thumb_03_l {parent=thumb_02_l, loc=<2.2223 0.3682 -0.0907>, rot=<0.0118 -0.013 0.0066 -0.9998>}
 index_metacarpal_l {parent=hand_l, loc=<1.6535 0.0183 -2.96>, rot=<0.0158 -0.0124 -0.0173 -0.9996>}
 index_01_l {parent=index_metacarpal_l, loc=<8.3008 0.0357 -0.0524>, rot=<-0.0044 0.0485 0.0924 -0.9945>}
 index_02_l {parent=index_01_l, loc=<3.3587 -0.3004 0.0452>, rot=<-0.002 0.0027 0.3225 -0.9466>}
 index_03_l {parent=index_02_l, loc=<2.5578 -0.123 -0.0252>, rot=<-0.002 0.0067 0.2558 -0.9667>}
 middle_metacarpal_l {parent=hand_l, loc=<2.1856 -0.1725 -0.5984>, rot=<0.006 0.0168 -0.0329 -0.9993>}
 middle_01_l {parent=middle_metacarpal_l, loc=<7.3275 -0.2447 0.0129>, rot=<-0.0208 0.0042 0.0858 -0.9961>}
 middle_02_l {parent=middle_01_l, loc=<4.0123 0.1401 0.1175>, rot=<-0.0218 0.0258 0.3512 -0.9357>}
 middle_03_l {parent=middle_02_l, loc=<2.6077 0.0534 0.1373>, rot=<0.0082 0.0188 0.1708 -0.9851>}
 ring_metacarpal_l {parent=hand_l, loc=<2.2074 -0.3625 1.3638>, rot=<-0.0284 0.0722 -0.0458 -0.9959>}
 ring_01_l {parent=ring_metacarpal_l, loc=<7.0295 -0.6197 -0.1446>, rot=<-0.0489 -0.0252 0.1892 -0.9804>}
 ring_02_l {parent=ring_01_l, loc=<3.6174 0.4647 -0.1027>, rot=<0 0 0.2806 -0.9598>}
 ring_03_l {parent=ring_02_l, loc=<2.4152 0.114 0.0199>, rot=<0 0 0.1189 -0.9929>}
 pinky_metacarpal_l {parent=hand_l, loc=<1.7616 -0.9753 3.3814>, rot=<-0.0397 0.0814 -0.069 -0.9935>}
 pinky_01_l {parent=pinky_metacarpal_l, loc=<7.1709 -0.4831 -0.1094>, rot=<-0.0654 -0.0035 0.2242 -0.9723>}
 pinky_02_l {parent=pinky_01_l, loc=<2.8274 0.0893 -0.1043>, rot=<0.0037 0.0011 0.2854 -0.9584>}
 pinky_03_l {parent=pinky_02_l, loc=<2.2046 0.0307 -0.0078>, rot=<-0.0002 0.0098 0.1445 -0.9895>}
 lowerarm_twist_02_l {parent=lowerarm_l, loc=<10.6707 0 0>, rot=<0 0 0 1>}
 lowerarm_twist_01_l {parent=lowerarm_l, loc=<18.316 0 0>, rot=<0 0 0 1>}
 dyn_fur_arm_l_A {parent=lowerarm_l, loc=<4.0011 -0.7351 3.647>, rot=<0.211 0.0245 0.9758 0.0516>}
 dyn_fur_arm_l_B {parent=lowerarm_l, loc=<-0.2818 -1.305 3.1827>, rot=<0.0529 -0.2469 0.9645 0.0769>}
 dyn_fur_arm_l_C {parent=lowerarm_l, loc=<-0.4976 -3.4866 0.8091>, rot=<-0.0188 -0.6551 0.745 0.1243>}
 dyn_fur_arm_l_D {parent=lowerarm_l, loc=<-0.959 -1.4011 -2.3695>, rot=<-0.1781 -0.9348 0.2122 -0.222>}
 upperarm_twist_01_l {parent=upperarm_l, loc=<7.2135 -0.1134 0>, rot=<0 0 0 1>}
 upperarm_twist_02_l {parent=upperarm_l, loc=<19.7998 -0.1134 0>, rot=<0 0 0 1>}
 clavicle_r {parent=spine_05, loc=<5.4831 -0.1053 3.2946>, rot=<0.7289 0.0465 -0.6829 0.009>}
 upperarm_r {parent=clavicle_r, loc=<-9.8658 1.2136 0.4105>, rot=<0.0164 -0.3245 -0.0269 -0.9454>}
 lowerarm_r {parent=upperarm_r, loc=<-26.465 0.8345 -0.259>, rot=<0.0036 0.0161 -0.2298 -0.9731>}
 hand_r {parent=lowerarm_r, loc=<-22.7355 0.3203 0.6106>, rot=<-0.6784 -0.0223 0.0735 -0.7307>}
 thumb_01_r {parent=hand_r, loc=<-1.3118 0.3191 2.459>, rot=<0.6546 -0.4042 0.0664 -0.6354>}
 thumb_02_r {parent=thumb_01_r, loc=<-5.9336 0.0448 -0.0401>, rot=<-0.0037 0.0213 0.298 -0.9543>}
 thumb_03_r {parent=thumb_02_r, loc=<-2.2223 -0.3682 0.0907>, rot=<0.0118 -0.013 0.0066 -0.9998>}
 index_metacarpal_r {parent=hand_r, loc=<-1.6535 -0.0183 2.96>, rot=<0.0158 -0.0124 -0.0173 -0.9996>}
 index_01_r {parent=index_metacarpal_r, loc=<-8.3008 -0.0357 0.0524>, rot=<-0.0044 0.0485 0.0924 -0.9945>}
 index_02_r {parent=index_01_r, loc=<-3.3587 0.3004 -0.0452>, rot=<-0.002 0.0027 0.3225 -0.9466>}
 index_03_r {parent=index_02_r, loc=<-2.5578 0.123 0.0252>, rot=<-0.002 0.0067 0.2558 -0.9667>}
 middle_metacarpal_r {parent=hand_r, loc=<-2.1856 0.1725 0.5984>, rot=<0.006 0.0168 -0.0329 -0.9993>}
 middle_01_r {parent=middle_metacarpal_r, loc=<-7.3275 0.2447 -0.0129>, rot=<-0.0208 0.0042 0.0858 -0.9961>}
 middle_02_r {parent=middle_01_r, loc=<-4.0123 -0.1401 -0.1175>, rot=<-0.0218 0.0258 0.3512 -0.9357>}
 middle_03_r {parent=middle_02_r, loc=<-2.6077 -0.0534 -0.1373>, rot=<0.0082 0.0188 0.1708 -0.9851>}
 ring_metacarpal_r {parent=hand_r, loc=<-2.2074 0.3625 -1.3638>, rot=<-0.0284 0.0722 -0.0458 -0.9959>}
 ring_01_r {parent=ring_metacarpal_r, loc=<-7.0295 0.6197 0.1446>, rot=<-0.0489 -0.0252 0.1892 -0.9804>}
 ring_02_r {parent=ring_01_r, loc=<-3.6174 -0.4647 0.1027>, rot=<0 0 0.2806 -0.9598>}
 ring_03_r {parent=ring_02_r, loc=<-2.4152 -0.114 -0.0199>, rot=<0 0 0.1189 -0.9929>}
 pinky_metacarpal_r {parent=hand_r, loc=<-1.7616 0.9753 -3.3814>, rot=<-0.0397 0.0814 -0.069 -0.9935>}
 pinky_01_r {parent=pinky_metacarpal_r, loc=<-7.1709 0.4831 0.1094>, rot=<-0.0654 -0.0035 0.2242 -0.9723>}
 pinky_02_r {parent=pinky_01_r, loc=<-2.8274 -0.0893 0.1043>, rot=<0.0037 0.0011 0.2854 -0.9584>}
 pinky_03_r {parent=pinky_02_r, loc=<-2.2046 -0.0307 0.0078>, rot=<-0.0002 0.0098 0.1445 -0.9895>}
 lowerarm_twist_02_r {parent=lowerarm_r, loc=<-10.6707 0 0>, rot=<0 0 0 1>}
 lowerarm_twist_01_r {parent=lowerarm_r, loc=<-18.316 0 0>, rot=<0 0 0 1>}
 dyn_fur_arm_r_A {parent=lowerarm_r, loc=<-3.0337 0.6645 -4.2489>, rot=<0.9809 -0.0893 -0.144 -0.0952>}
 dyn_fur_arm_r_B {parent=lowerarm_r, loc=<1.2649 0.8101 -3.2188>, rot=<0.9748 -0.0263 -0.0124 -0.2213>}
 dyn_fur_arm_r_C {parent=lowerarm_r, loc=<1.3605 3.5754 -0.5196>, rot=<0.8195 -0.2442 0.0004 -0.5184>}
 dyn_fur_arm_r_D {parent=lowerarm_r, loc=<1.7221 0.5644 2.828>, rot=<0.1997 0.2415 0.0959 -0.9448>}
 upperarm_twist_01_r {parent=upperarm_r, loc=<-7.2135 0.1134 0>, rot=<0 0 0 1>}
 upperarm_twist_02_r {parent=upperarm_r, loc=<-19.7998 0.1134 0>, rot=<0 0 0 1>}
 attach_backpack {parent=spine_05, loc=<-0.123 -13.473 0>, rot=<-0.0093 -0.707 0.0093 -0.707>}
 attach_cape {parent=spine_05, loc=<9.017 -11.714 0>, rot=<0.0616 -0.7044 -0.0616 -0.7044>}
 dyn_fur_hip_r_A {parent=spine_01, loc=<6.547 0.912 9.153>, rot=<0.3825 -0.0105 0.9222 -0.0552>}
 dyn_fur_hip_r_B {parent=spine_01, loc=<5.4099 -1.7833 8.7813>, rot=<0.3412 -0.1217 0.9307 -0.0506>}
 dyn_fur_hip_r_C {parent=spine_01, loc=<4.6942 2.6635 9.153>, rot=<0.4164 0.1568 0.8945 0.0434>}
 dyn_fur_hip_l_A {parent=spine_01, loc=<6.4199 0.9079 -9.1528>, rot=<0.0793 -0.9205 0.0205 -0.3821>}
 dyn_fur_hip_l_B {parent=spine_01, loc=<4.1795 2.7786 -9.5053>, rot=<-0.0225 -0.8581 -0.1395 -0.4936>}
 dyn_fur_hip_l_C {parent=spine_01, loc=<4.5519 0.5788 -9.5053>, rot=<-0.0253 -0.9202 -0.0233 -0.39>}
 dyn_fur_hip_l_D {parent=spine_01, loc=<4.5856 -0.9571 -9.3457>, rot=<0.0853 -0.9333 0.1854 -0.2954>}
 thigh_l {parent=pelvis, loc=<-1.0855 -0.2109 -8.7053>, rot=<0.0009 0.0355 0.0155 -0.9993>}
 calf_l {parent=thigh_l, loc=<-42.5284 -0.0019 -0.7297>, rot=<0.0006 -0.0161 -0.0391 -0.9991>}
 foot_l {parent=calf_l, loc=<-46.7401 -2.5014 -0.592>, rot=<0.0427 -0.0238 0.0353 -0.9982>}
 ball_l {parent=foot_l, loc=<-3.5658 10.8066 -0.1544>, rot=<0.0001 -0.0001 -0.7063 -0.7079>}
 calf_twist_02_l {parent=calf_l, loc=<-19.1546 -1.7943 -0.2638>, rot=<0.0237 0.0095 -0.0035 -0.9997>}
 calf_twist_01_l {parent=calf_l, loc=<-35.8483 -1.9197 -0.5794>, rot=<0.0237 0.0095 -0.0035 -0.9997>}
 dyn_pants_l_base {parent=calf_l, loc=<-28.8176 -1.4136 -0.6817>, rot=<0 -0.7071 -0.7071 0>}
 dyn_pants_l_fr_01 {parent=dyn_pants_l_base, loc=<-0.2418 -0.5863 3.2619>, rot=<-0.0407 0.0119 -0.0198 -0.9989>}
 dyn_pants_l_fr_02 {parent=dyn_pants_l_fr_01, loc=<8.6591 0 0>, rot=<0 0 0 1>}
 dyn_pants_l_inner_01 {parent=dyn_pants_l_base, loc=<-0.1914 3.4245 -0.2752>, rot=<0.6773 0.0281 -0.0005 -0.7351>}
 dyn_pants_l_inner_02 {parent=dyn_pants_l_inner_01, loc=<10.6863 0 0>, rot=<0.0436 0 0 -0.999>}
 dyn_pants_l_bck_01 {parent=dyn_pants_l_base, loc=<0.4687 -0.6125 -5.2857>, rot=<0.9997 0.0192 -0.0114 0.0033>}
 dyn_pants_l_bck_02 {parent=dyn_pants_l_bck_01, loc=<8.6999 0 -1.8234>, rot=<0 0 0 1>}
 dyn_pants_l_outer_01 {parent=dyn_pants_l_base, loc=<0.1197 -4.6433 -0.2247>, rot=<-0.73 -0.0119 0.0416 -0.6821>}
 dyn_pants_l_outer_02 {parent=dyn_pants_l_outer_01, loc=<10.2741 0 0>, rot=<-0.0436 0.0046 -0.0346 -0.9984>}
 thigh_twist_01_l {parent=thigh_l, loc=<-27.952 0 0>, rot=<0.0013 0 0 -1>}
 dyn_strap_fr_l_base {parent=thigh_l, loc=<1.978 6.2722 0.6619>, rot=<-0.0967 -0.987 0.1271 -0.017>}
 dyn_strap_fr_l_01 {parent=dyn_strap_fr_l_base, loc=<0 0 0>, rot=<0 0 0 1>}
 dyn_strap_fr_l_02 {parent=dyn_strap_fr_l_01, loc=<4.3851 0 0>, rot=<-0.0175 0.049 0.1346 -0.9895>}
 dyn_strap_fr_l_03 {parent=dyn_strap_fr_l_02, loc=<2.9429 0 0>, rot=<0.0435 0.0551 -0.106 -0.9919>}
 dyn_strap_fr_l_04 {parent=dyn_strap_fr_l_03, loc=<8.1354 0 0>, rot=<-0.0192 0.107 0.0406 -0.9932>}
 dyn_strap_fr_l_05 {parent=dyn_strap_fr_l_04, loc=<8.0798 -1.4238 0.0497>, rot=<-0.1179 0.1042 0.2196 -0.9628>}
 dyn_strap_l_c {parent=thigh_l, loc=<-24.2742 0.0972 -6.8689>, rot=<-0.002 -0.7884 0.0014 -0.6152>}
 thigh_r {parent=pelvis, loc=<-1.0855 -0.2109 8.7053>, rot=<0.0355 -0.0009 0.9993 0.0155>}
 calf_r {parent=thigh_r, loc=<42.5284 0.0019 0.7297>, rot=<0.0006 -0.0161 -0.0391 -0.9991>}
 foot_r {parent=calf_r, loc=<46.7401 2.5014 0.592>, rot=<0.0427 -0.0238 0.0353 -0.9982>}
 ball_r {parent=foot_r, loc=<3.5658 -10.8066 0.1544>, rot=<0.0001 -0.0001 -0.7063 -0.7079>}
 calf_twist_02_r {parent=calf_r, loc=<19.1546 1.7943 0.2638>, rot=<0.0237 0.0095 -0.0035 -0.9997>}
 calf_twist_01_r {parent=calf_r, loc=<35.8483 1.9197 0.5794>, rot=<0.0237 0.0095 -0.0035 -0.9997>}
 dyn_pants_r_base {parent=calf_r, loc=<28.8176 1.4136 0.6814>, rot=<-0.7071 0 0 -0.7071>}
 dyn_pants_r_fr_01 {parent=dyn_pants_r_base, loc=<-0.2418 0.5869 3.2619>, rot=<-0.0023 0.0128 -0.0243 -0.9996>}
 dyn_pants_r_fr_02 {parent=dyn_pants_r_fr_01, loc=<8.6262 -0.7547 0>, rot=<0.0436 -0.0019 0.0436 -0.9981>}
 dyn_pants_r_inner_01 {parent=dyn_pants_r_base, loc=<-0.1914 -3.4245 -0.2752>, rot=<-0.6107 0.028 0.0029 -0.7914>}
 dyn_pants_r_inner_02 {parent=dyn_pants_r_inner_01, loc=<10.6863 0 0>, rot=<-0.1305 0 0 -0.9914>}
 dyn_pants_r_bck_01 {parent=dyn_pants_r_base, loc=<0.4687 0.6131 -5.2857>, rot=<0.9989 -0.0187 -0.0122 0.0403>}
 dyn_pants_r_bck_02 {parent=dyn_pants_r_bck_01, loc=<10.1941 0 0>, rot=<0 0 0 1>}
 dyn_pants_r_outer_01 {parent=dyn_pants_r_base, loc=<0.1197 4.6436 -0.2247>, rot=<0.73 -0.0119 -0.0415 -0.6821>}
 dyn_pants_r_outer_02 {parent=dyn_pants_r_outer_01, loc=<10.2741 0 0>, rot=<0 -0.0076 0.043 -0.999>}
 thigh_twist_01_r {parent=thigh_r, loc=<27.952 0 0>, rot=<0.0013 0 0 -1>}
 dyn_strap_fr_r_base {parent=thigh_r, loc=<-1.978 -6.2722 -0.6621>, rot=<-0.0029 0.0236 0.0773 -0.9967>}
 dyn_strap_fr_r_01 {parent=dyn_strap_fr_r_base, loc=<0 0 0>, rot=<0 0 0 1>}
 dyn_strap_fr_r_02 {parent=dyn_strap_fr_r_01, loc=<4.3824 -0.1438 0.0523>, rot=<-0.0041 0.0153 -0.1267 -0.9918>}
 dyn_strap_fr_r_03 {parent=dyn_strap_fr_r_02, loc=<2.9429 0 0>, rot=<0.0588 0.0333 0.1217 -0.9903>}
 dyn_strap_fr_r_04 {parent=dyn_strap_fr_r_03, loc=<8.0935 0.4258 0.7081>, rot=<0.1025 0.0845 -0.0481 -0.99>}
 dyn_strap_fr_r_05 {parent=dyn_strap_fr_r_04, loc=<7.9568 1.5415 1.2735>, rot=<0.1503 0.1139 -0.2204 -0.957>}
 dyn_strap_r_c {parent=thigh_r, loc=<24.2742 -0.0972 6.8685>, rot=<0.0013 0.5802 0.0021 -0.8145>}
 dyn_strap_bck_r_base {parent=pelvis, loc=<5.9642 -7.2677 7.506>, rot=<0.134 0.0596 0.9702 -0.1928>}
 dyn_strap_bck_r_01 {parent=dyn_strap_bck_r_base, loc=<0 0 0>, rot=<0 0 0 1>}
 dyn_strap_bck_r_02 {parent=dyn_strap_bck_r_01, loc=<6.7896 0 0>, rot=<-0.1717 -0.0246 0.1395 -0.9749>}
 dyn_strap_bck_r_03 {parent=dyn_strap_bck_r_02, loc=<6.823 -0.1632 0.4483>, rot=<-0.0681 0.1028 0.1075 -0.9865>}
 dyn_strap_bck_r_04 {parent=dyn_strap_bck_r_03, loc=<9.1162 0 0>, rot=<-0.1008 0.023 0.0237 -0.9944>}
 dyn_strap_bck_r_05 {parent=dyn_strap_bck_r_04, loc=<8.6411 -0.3018 1.8378>, rot=<-0.0031 0.2475 0.1179 -0.9617>}
 dyn_strap_bck_l_base {parent=pelvis, loc=<5.9642 -7.2677 -7.5056>, rot=<0.2027 -0.9625 -0.1355 -0.1185>}
 dyn_strap_bck_l_01 {parent=dyn_strap_bck_l_base, loc=<0 0 0>, rot=<0 0 0 1>}
 dyn_strap_bck_l_02 {parent=dyn_strap_bck_l_01, loc=<6.7896 0 0>, rot=<0.0324 0.0085 -0.1566 -0.9871>}
 dyn_strap_bck_l_03 {parent=dyn_strap_bck_l_02, loc=<6.8397 0 0>, rot=<0.1355 0.0129 -0.0944 -0.9862>}
 dyn_strap_bck_l_04 {parent=dyn_strap_bck_l_03, loc=<9.0664 0 0.9529>, rot=<0.0684 0.0893 -0.0449 -0.9926>}
 dyn_strap_bck_l_05 {parent=dyn_strap_bck_l_04, loc=<8.7051 0 1.5349>, rot=<0.0226 0.2578 -0.0842 -0.9623>}
]
variables [
 L FootBend Alpha {type=double, default=0}
 L calfBend Alpha {type=float, default=0}
 L thighForward Alpha {type=double, default=0}
 R FootBend Alpha {type=double, default=0}
 R calfBend Alpha {type=float, default=0}
 R thighForward Alpha {type=double, default=0}
]
