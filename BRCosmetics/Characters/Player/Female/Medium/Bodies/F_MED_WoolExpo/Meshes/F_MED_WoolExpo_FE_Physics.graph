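# Physics asset for the WoolExpo female medium player body mesh that simulates dangling straps on each side of the lower body.
# Pelvis, thigh and calf capsules provide leg collision, with back strap chains anchored to the pelvis and front chains to the thighs through free-rotation base constraints.
# Each five-link capsule chain allows limited swing and minimal twist.
# Front and back chains join at a shared center bone per side, forming closed loops.
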PhysicsAsset /BRCosmetics/Characters/Player/Female/Medium/Bodies/F_MED_WoolExpo/Meshes/F_MED_WoolExpo_FE_Physics
bodies {
 calf_l [{
   shapeType=Capsule
   length=42.8296
   radius=8.4423
   shapeName=calf_l_capsule
   center=<-23.1186 -1.261 -1.0357>
   rotation=<rot -0.2139 92.1645 -88.7867>
  }]
 calf_r [{
   shapeType=Capsule
   length=39.6398
   radius=8.2188
   shapeName=calf_r_capsule
   center=<22.3286 2.2531 -0.1452>
   rotation=<rot 0 92.6601 -88.8331>
  }]
 dyn_strap_bck_l_01 [{shapeType=Capsule, length=3.5, radius=1.5, shapeName=dyn_strap_bck_l_01_capsule, center=<3.5576 -0.5949 0>, rotation=<rot -90 0.0001 0>}]
 dyn_strap_bck_l_02 [{shapeType=Capsule, length=3.5, radius=1.5, shapeName=dyn_strap_bck_l_02_capsule, center=<2.513 -0.7468 0>, rotation=<rot -90 0.0002 0>}]
 dyn_strap_bck_l_03 [{shapeType=Capsule, length=5.5, radius=1.5, shapeName=dyn_strap_bck_l_03_capsule, center=<3.2816 -0.89 0>, rotation=<rot -90 0.0001 0>}]
 dyn_strap_bck_l_04 [{shapeType=Capsule, length=5, radius=2, shapeName=dyn_strap_bck_l_04_capsule, center=<2.513 -1.1051 0>, rotation=<rot -90 0.0001 0>}]
 dyn_strap_bck_l_05 [{shapeType=Capsule, length=3.5, radius=2, shapeName=dyn_strap_bck_l_05_capsule, center=<1.5851 -1.1052 0>, rotation=<rot -90 0 0>}]
 dyn_strap_bck_l_base [{shapeType=Capsule, length=0.505, radius=0.505, shapeName=dyn_strap_bck_l_base_capsule, rotation=<rot 0 -90 0>}]
 dyn_strap_bck_r_01 [{shapeType=Capsule, length=3.5, radius=1.5, shapeName=dyn_strap_bck_r_01_capsule, center=<2.2934 0.9704 0>, rotation=<rot -90 -0.0001 0>}]
 dyn_strap_bck_r_02 [{shapeType=Capsule, length=4, radius=1.5, shapeName=dyn_strap_bck_r_02_capsule, center=<1.649 0.9704 0>, rotation=<rot -90 -0.0002 0>}]
 dyn_strap_bck_r_03 [{shapeType=Capsule, length=4.5, radius=1.5, shapeName=dyn_strap_bck_r_03_capsule, center=<2.3983 0.9704 0>, rotation=<rot -90 -0.0001 0>}]
 dyn_strap_bck_r_04 [{shapeType=Capsule, length=4, radius=2, shapeName=dyn_strap_bck_r_04_capsule, center=<1.5246 0.6352 0>, rotation=<rot -90 -0.0001 0>}]
 dyn_strap_bck_r_05 [{shapeType=Capsule, length=3.5, radius=2.5, shapeName=dyn_strap_bck_r_05_capsule, center=<1.5246 0.9704 0>, rotation=<rot -90 0 0>}]
 dyn_strap_bck_r_base [{shapeType=Capsule, length=0.505, radius=0.505, shapeName=dyn_strap_bck_r_base_capsule, rotation=<rot 0 -90 0>}]
 dyn_strap_fr_l_01 [{shapeType=Capsule, length=4, radius=2, shapeName=dyn_strap_fr_l_01_capsule, center=<3.4421 1.8014 0>, rotation=<rot -90 0.0002 0>}]
 dyn_strap_fr_l_02 [{shapeType=Capsule, length=1.5, radius=2, shapeName=dyn_strap_fr_l_02_capsule, center=<2.1523 2.2087 0>, rotation=<rot -90 10.0002 0>}]
 dyn_strap_fr_l_03 [{shapeType=Capsule, length=3.5, radius=2, shapeName=dyn_strap_fr_l_03_capsule, center=<3.4421 1.5354 0>, rotation=<rot -90 0.0001 0>}]
 dyn_strap_fr_l_04 [{shapeType=Capsule, length=5, radius=2, shapeName=dyn_strap_fr_l_04_capsule, center=<3.6144 0.6763 0>, rotation=<rot -90 -9.9999 0>}]
 dyn_strap_fr_l_05 [{shapeType=Capsule, length=3, radius=2.5, shapeName=dyn_strap_fr_l_05_capsule, center=<3.7026 1.1763 0>, rotation=<rot -90 -9.9999 0>}]
 dyn_strap_fr_l_base [{shapeType=Capsule, length=0.505, radius=0.505, shapeName=dyn_strap_fr_l_base_capsule, rotation=<rot 0 -90 0>}]
 dyn_strap_fr_r_01 [{shapeType=Capsule, length=4, radius=2, shapeName=dyn_strap_fr_r_01_capsule, center=<1.613 -1.8428 0>, rotation=<rot -90 -0.0002 0>}]
 dyn_strap_fr_r_02 [{shapeType=Capsule, length=2, radius=2, shapeName=dyn_strap_fr_r_02_capsule, center=<1.2685 -2.0948 0>, rotation=<rot -90 -10.0003 0>}]
 dyn_strap_fr_r_03 [{shapeType=Capsule, length=3, radius=2, shapeName=dyn_strap_fr_r_03_capsule, center=<2.9329 -1.5052 0>, rotation=<rot -90 -0.0002 0>}]
 dyn_strap_fr_r_04 [{shapeType=Capsule, length=5, radius=2, shapeName=dyn_strap_fr_r_04_capsule, center=<2.2291 -0.9091 0>, rotation=<rot -90 9.9999 0>}]
 dyn_strap_fr_r_05 [{
   shapeType=Capsule
   length=3
   radius=2
   shapeName=dyn_strap_fr_r_05_capsule
   center=<1.8795 -1.5347 0.3314>
   rotation=<rot -80.0001 -0.0003 0.0002>
  }]
 dyn_strap_fr_r_base [{shapeType=Capsule, length=0.505, radius=0.505, shapeName=dyn_strap_fr_r_base_capsule, rotation=<rot 0 -90 0>}]
 dyn_strap_l_c [{shapeType=Capsule, length=4.5, radius=2, shapeName=dyn_strap_l_c_capsule, center=<1.5533 0 0>, rotation=<rot 0 0 -90>}]
 dyn_strap_r_c [{shapeType=Capsule, length=5, radius=2.5, shapeName=dyn_strap_r_c_capsule, center=<2.1211 0 -0.1948>, rotation=<rot 0 0 -90>}]
 pelvis [{
   shapeType=Capsule
   length=16.4508
   radius=7.7404
   shapeName=pelvis_capsule
   center=<-2.0036 -2.1135 -0.0548>
   rotation=<rot -0.8689 -90 0.3152>
  }]
 thigh_l [{shapeType=Capsule, length=31, radius=8.5, shapeName=thigh_l_capsule, center=<-16.9293 -1.3406 1.7455>, rotation=<rot 90 0 0>}]
 thigh_r [{shapeType=Capsule, length=31, radius=8.5, shapeName=thigh_r_capsule, center=<16.9453 1.292 -1.4619>, rotation=<rot 90 0 0>}]
}
constraints [
 {bone1Name=dyn_strap_bck_l_01, bone2Name=dyn_strap_bck_l_base, swing1LimitDegrees=60, swing2LimitDegrees=60, twistLimitDegrees=1}
 {bone1Name=dyn_strap_bck_l_02, bone2Name=dyn_strap_bck_l_01, swing1LimitDegrees=45, swing2LimitDegrees=45, twistLimitDegrees=1}
 {bone1Name=dyn_strap_bck_l_03, bone2Name=dyn_strap_bck_l_02, swing1LimitDegrees=60, swing2LimitDegrees=60, twistLimitDegrees=1}
 {bone1Name=dyn_strap_bck_l_04, bone2Name=dyn_strap_bck_l_03, swing1LimitDegrees=60, swing2LimitDegrees=60, twistLimitDegrees=1}
 {bone1Name=dyn_strap_bck_l_05, bone2Name=dyn_strap_bck_l_04, swing1LimitDegrees=60, swing2LimitDegrees=60, twistLimitDegrees=1}
 {bone1Name=dyn_strap_bck_r_01, bone2Name=dyn_strap_bck_r_base, swing1LimitDegrees=60, swing2LimitDegrees=60, twistLimitDegrees=1}
 {bone1Name=dyn_strap_bck_r_02, bone2Name=dyn_strap_bck_r_01, swing1LimitDegrees=45, swing2LimitDegrees=45, twistLimitDegrees=1}
 {bone1Name=dyn_strap_bck_r_03, bone2Name=dyn_strap_bck_r_02, swing1LimitDegrees=60, swing2LimitDegrees=60, twistLimitDegrees=1}
 {bone1Name=dyn_strap_bck_r_04, bone2Name=dyn_strap_bck_r_03, swing1LimitDegrees=60, swing2LimitDegrees=60, twistLimitDegrees=1}
 {bone1Name=dyn_strap_bck_r_05, bone2Name=dyn_strap_bck_r_04, swing1LimitDegrees=60, swing2LimitDegrees=60, twistLimitDegrees=1}
 {
  bone1Name=dyn_strap_bck_r_base
  bone2Name=pelvis
  swing1LimitDegrees=45
  swing1Motion=Free
  swing2LimitDegrees=45
  swing2Motion=Free
  twistLimitDegrees=1
  twistMotion=Free
 }
 {
  bone1Name=dyn_strap_bck_l_base
  bone2Name=pelvis
  swing1LimitDegrees=45
  swing1Motion=Free
  swing2LimitDegrees=45
  swing2Motion=Free
  twistLimitDegrees=1
  twistMotion=Free
 }
 {
  bone1Name=dyn_strap_fr_l_base
  bone2Name=thigh_l
  swing1LimitDegrees=45
  swing1Motion=Free
  swing2LimitDegrees=45
  swing2Motion=Free
  twistLimitDegrees=45
  twistMotion=Free
 }
 {bone1Name=dyn_strap_fr_l_01, bone2Name=dyn_strap_fr_l_base, swing1LimitDegrees=45, swing2LimitDegrees=45, twistLimitDegrees=1}
 {bone1Name=dyn_strap_fr_l_02, bone2Name=dyn_strap_fr_l_01, swing1LimitDegrees=60, swing2LimitDegrees=60, twistLimitDegrees=1}
 {bone1Name=dyn_strap_fr_l_03, bone2Name=dyn_strap_fr_l_02, swing1LimitDegrees=60, swing2LimitDegrees=60, twistLimitDegrees=1}
 {bone1Name=dyn_strap_fr_l_04, bone2Name=dyn_strap_fr_l_03, swing1LimitDegrees=60, swing2LimitDegrees=60, twistLimitDegrees=1}
 {bone1Name=dyn_strap_fr_l_05, bone2Name=dyn_strap_fr_l_04, swing1LimitDegrees=60, swing2LimitDegrees=60, twistLimitDegrees=1}
 {
  bone1Name=dyn_strap_fr_r_base
  bone2Name=thigh_r
  swing1LimitDegrees=45
  swing1Motion=Free
  swing2LimitDegrees=45
  swing2Motion=Free
  twistLimitDegrees=1
  twistMotion=Free
 }
 {bone1Name=dyn_strap_fr_r_01, bone2Name=dyn_strap_fr_r_base, swing1LimitDegrees=45, swing2LimitDegrees=45, twistLimitDegrees=1}
 {bone1Name=dyn_strap_fr_r_02, bone2Name=dyn_strap_fr_r_01, swing1LimitDegrees=60, swing2LimitDegrees=60, twistLimitDegrees=1}
 {bone1Name=dyn_strap_fr_r_03, bone2Name=dyn_strap_fr_r_02, swing1LimitDegrees=60, swing2LimitDegrees=60, twistLimitDegrees=1}
 {bone1Name=dyn_strap_fr_r_04, bone2Name=dyn_strap_fr_r_03, swing1LimitDegrees=60, swing2LimitDegrees=60, twistLimitDegrees=1}
 {bone1Name=dyn_strap_fr_r_05, bone2Name=dyn_strap_fr_r_04, swing1LimitDegrees=60, swing2LimitDegrees=60, twistLimitDegrees=1}
 {bone1Name=dyn_strap_l_c, bone2Name=dyn_strap_bck_l_05, swing1LimitDegrees=45, swing2LimitDegrees=45, twistLimitDegrees=1}
 {bone1Name=dyn_strap_l_c, bone2Name=dyn_strap_fr_l_05, swing1LimitDegrees=45, swing2LimitDegrees=45, twistLimitDegrees=1}
 {bone1Name=dyn_strap_r_c, bone2Name=dyn_strap_fr_r_05, swing1LimitDegrees=45, swing2LimitDegrees=45, twistLimitDegrees=1}
 {bone1Name=dyn_strap_r_c, bone2Name=dyn_strap_bck_r_05, swing1LimitDegrees=45, swing2LimitDegrees=45, twistLimitDegrees=1}
]
