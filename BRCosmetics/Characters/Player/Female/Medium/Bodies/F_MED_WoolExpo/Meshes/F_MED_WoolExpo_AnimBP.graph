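# Anim Blueprint for the F_MED_WoolExpo cosmetic body, adding secondary motion to straps, fur and pants.
# A parent-copied pose runs through Control Rig, RigidBody, strap constraints, per-bone AnimDynamics for fur and pants, pant-to-foot constraints, then a front-end RigidBody.
# The event graph updates gravity and sim-space variables from a parameter asset.
# Emote, customization-load and skydiving events set front-end versus in-game physics alphas, AnimDynamics space and pant spring stiffness.
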AnimBlueprint /BRCosmetics/Characters/Player/Female/Medium/Bodies/F_MED_WoolExpo/Meshes/F_MED_WoolExpo_AnimBP
target_skeleton=&/F_MED_WoolExpo_Skeleton
graphs [
 AnimGraph {nodes [
   n1 = Control Rig(Source=Copy Pose From Mesh() {bUseAttachedParent=true}) {
    alphaBoolBlend {blendOption=Linear}
    controlRigAssetReference {blueprintRigClass=&/F_MED_WoolExpo_CtrlRig.F_MED_WoolExpo_CtrlRig_C}
   }
   n2 = RigidBody!(
    ComponentPose=Local To Component(LocalPose=@n1)
    Alpha
    OverrideWorldGravity:Vector="0,0,-980"
    ExternalForce
    SimSpaceSettings:Sim Space Settings Structure=(WorldAlpha=0.25,VelocityScaleZ=1,DampingAlpha=1,MaxLinearVelocity=800,MaxAngularVelocity=8,MaxLinearAcceleration=8000,MaxAngularAcceleration=8000,ExternalLinearDragV=(X=0,Y=0,Z=0),ExternalLinearVelocity=(X=0,Y=0,Z=0),ExternalAngularVelocity=(X=0,Y=0,Z=0))
   ) -> (Pose) {
    bClampLinearTranslationLimitToRefPose=true
    bOverrideWorldGravity=true
    overridePhysicsAsset=&/F_MED_WoolExpo_Physics
    simulationSpace=BaseBoneSpace
    alphaBoolBlend {blendOption=Linear}
    alphaScaleBiasClamp {bInterpResult=true}
    baseBoneRef {boneName=pelvis}
    overrideWorldGravity {z=-980}
    simSpaceSettings {maxAngularAcceleration=8000, maxAngularVelocity=8, maxLinearAcceleration=8000, maxLinearVelocity=800, worldAlpha=0.25}
   }
   n3 = "Constraint - dyn_strap_l_c (PP)"!(ComponentPose=@n2, Alpha=1, ConstraintWeights_0=0.5, ConstraintWeights_1:float=0.5) -> (Pose) {
    alphaBoolBlend {blendOption=Linear}
    boneToModify {boneName=dyn_strap_l_c}
    constraintSetup [{targetBone {boneName=dyn_strap_fr_l_05}}, {targetBone {boneName=dyn_strap_bck_l_05}}]
    constraintWeights [1,1]
   }
   n4 = "Constraint - dyn_strap_r_c (PP)"!(ComponentPose=@n3, Alpha=1, ConstraintWeights_0=0.5, ConstraintWeights_1:float=0.5) -> (Pose) {
    alphaBoolBlend {blendOption=Linear}
    boneToModify {boneName=dyn_strap_r_c}
    constraintSetup [{targetBone {boneName=dyn_strap_fr_r_05}}, {targetBone {boneName=dyn_strap_bck_r_05}}]
    constraintWeights [1,1]
   }
   n5 = Anim Dynamics - Bone: dyn_fur_hip_r_A(ComponentPose=@n4, Alpha=1, ExternalForce="0,0,0") {
    angularDampingOverride=0.99
    angularSpringConstant=5000
    bAngularSpring=true
    bGravityOverrideInSimSpace=true
    bOverrideAngularDamping=true
    bUseGravityOverride=true
    simulationSpace=BoneRelative
    alphaBoolBlend {blendOption=Linear}
    boundBone {boneName=dyn_fur_hip_r_A}
    componentAppliedLinearAccClamp=<1000 1000 1000>
    componentLinearAccScale=<1 1 1>
    componentLinearVelScale=<1 1 1>
    gravityOverride {z=-980}
    physicsBodyDefinitions [{
      boundBone {boneName=dyn_fur_hip_r_A}
      boxExtents=<18 18 18>
      constraintSetup {angularLimitsMax=<10 10 10>, angularLimitsMin=<-10 -30 -10>, angularTarget=<1 0 0>}
      localJointOffset=<1 0 0>
     }]
    relativeSpaceBone {boneName=root}
   }
   n6 = Anim Dynamics - Bone: dyn_fur_hip_r_B(ComponentPose=@n5, Alpha=1, ExternalForce="0,0,0") {
    angularDampingOverride=0.99
    angularSpringConstant=5000
    bAngularSpring=true
    bGravityOverrideInSimSpace=true
    bOverrideAngularDamping=true
    bUseGravityOverride=true
    simulationSpace=BoneRelative
    alphaBoolBlend {blendOption=Linear}
    boundBone {boneName=dyn_fur_hip_r_B}
    componentAppliedLinearAccClamp=<1000 1000 1000>
    componentLinearAccScale=<1 1 1>
    componentLinearVelScale=<1 1 1>
    gravityOverride {z=-980}
    physicsBodyDefinitions [{
      boundBone {boneName=dyn_fur_hip_r_B}
      boxExtents=<18 18 18>
      constraintSetup {angularLimitsMax=<10 10 10>, angularLimitsMin=<-10 -30 -10>, angularTarget=<1 0 0>}
      localJointOffset=<1 0 0>
     }]
    relativeSpaceBone {boneName=root}
   }
   n7 = Anim Dynamics - Bone: dyn_fur_hip_r_C(ComponentPose=@n6, Alpha=1, ExternalForce="0,0,0") {
    angularDampingOverride=0.99
    angularSpringConstant=5000
    bAngularSpring=true
    bGravityOverrideInSimSpace=true
    bOverrideAngularDamping=true
    simulationSpace=BoneRelative
    alphaBoolBlend {blendOption=Linear}
    boundBone {boneName=dyn_fur_hip_r_C}
    componentAppliedLinearAccClamp=<1000 1000 1000>
    componentLinearAccScale=<1 1 1>
    componentLinearVelScale=<1 1 1>
    gravityOverride {z=-980}
    physicsBodyDefinitions [{
      boundBone {boneName=dyn_fur_hip_r_C}
      boxExtents=<18 18 18>
      constraintSetup {angularLimitsMax=<10 10 10>, angularLimitsMin=<-10 -30 -10>, angularTarget=<1 0 0>}
      localJointOffset=<1 0 0>
     }]
    relativeSpaceBone {boneName=root}
   }
   n8 = Anim Dynamics - Bone: dyn_fur_hip_l_A(ComponentPose=@n7, Alpha=1, ExternalForce="0,0,0") {
    angularDampingOverride=0.99
    angularSpringConstant=5000
    bAngularSpring=true
    bGravityOverrideInSimSpace=true
    bOverrideAngularDamping=true
    bUseGravityOverride=true
    simulationSpace=BoneRelative
    alphaBoolBlend {blendOption=Linear}
    boundBone {boneName=dyn_fur_hip_l_A}
    componentAppliedLinearAccClamp=<1000 1000 1000>
    componentLinearAccScale=<1 1 1>
    componentLinearVelScale=<1 1 1>
    gravityOverride {z=-980}
    physicsBodyDefinitions [{
      boundBone {boneName=dyn_fur_hip_l_A}
      boxExtents=<18 18 18>
      constraintSetup {angularLimitsMax=<10 10 10>, angularLimitsMin=<-10 -30 -10>, angularTarget=<1 0 0>}
      localJointOffset=<1 0 0>
     }]
    relativeSpaceBone {boneName=root}
   }
   n9 = Anim Dynamics - Bone: dyn_fur_hip_l_B(ComponentPose=@n8, Alpha=1, ExternalForce="0,0,0") {
    angularDampingOverride=0.99
    angularSpringConstant=5000
    bAngularSpring=true
    bGravityOverrideInSimSpace=true
    bOverrideAngularDamping=true
    bUseGravityOverride=true
    simulationSpace=BoneRelative
    alphaBoolBlend {blendOption=Linear}
    boundBone {boneName=dyn_fur_hip_l_B}
    componentAppliedLinearAccClamp=<1000 1000 1000>
    componentLinearAccScale=<1 1 1>
    componentLinearVelScale=<1 1 1>
    gravityOverride {z=-980}
    physicsBodyDefinitions [{
      boundBone {boneName=dyn_fur_hip_l_B}
      boxExtents=<18 18 18>
      constraintSetup {angularLimitsMax=<10 10 10>, angularLimitsMin=<-10 -30 -10>, angularTarget=<1 0 0>}
      localJointOffset=<1 0 0>
     }]
    relativeSpaceBone {boneName=root}
   }
   n10 = Anim Dynamics - Bone: dyn_fur_hip_l_C(ComponentPose=@n9, Alpha=1, ExternalForce="0,0,0") {
    angularDampingOverride=0.99
    angularSpringConstant=5000
    bAngularSpring=true
    bGravityOverrideInSimSpace=true
    bOverrideAngularDamping=true
    simulationSpace=BoneRelative
    alphaBoolBlend {blendOption=Linear}
    boundBone {boneName=dyn_fur_hip_l_C}
    componentAppliedLinearAccClamp=<1000 1000 1000>
    componentLinearAccScale=<1 1 1>
    componentLinearVelScale=<1 1 1>
    gravityOverride {z=-980}
    physicsBodyDefinitions [{
      boundBone {boneName=dyn_fur_hip_l_C}
      boxExtents=<18 18 18>
      constraintSetup {angularLimitsMax=<10 10 10>, angularLimitsMin=<-10 -30 -10>, angularTarget=<1 0 0>}
      localJointOffset=<1 0 0>
     }]
    relativeSpaceBone {boneName=root}
   }
   n11 = Anim Dynamics - Bone: dyn_fur_hip_l_D(ComponentPose=@n10, Alpha=1, ExternalForce="0,0,0") {
    angularDampingOverride=0.99
    angularSpringConstant=5000
    bAngularSpring=true
    bGravityOverrideInSimSpace=true
    bOverrideAngularDamping=true
    simulationSpace=BoneRelative
    alphaBoolBlend {blendOption=Linear}
    boundBone {boneName=dyn_fur_hip_l_D}
    componentAppliedLinearAccClamp=<1000 1000 1000>
    componentLinearAccScale=<1 1 1>
    componentLinearVelScale=<1 1 1>
    gravityOverride {z=-980}
    physicsBodyDefinitions [{
      boundBone {boneName=dyn_fur_hip_l_D}
      boxExtents=<18 18 18>
      constraintSetup {angularLimitsMax=<10 10 10>, angularLimitsMin=<-10 -30 -10>, angularTarget=<1 0 0>}
      localJointOffset=<1 0 0>
     }]
    relativeSpaceBone {boneName=root}
   }
   n12 = Anim Dynamics - Bone: dyn_fur_arm_l_A(ComponentPose=@n11, Alpha=1, ExternalForce="0,0,0") {
    angularDampingOverride=0.99
    angularSpringConstant=5000
    bAngularSpring=true
    bGravityOverrideInSimSpace=true
    bOverrideAngularDamping=true
    bUseGravityOverride=true
    simulationSpace=BoneRelative
    alphaBoolBlend {blendOption=Linear}
    boundBone {boneName=dyn_fur_arm_l_A}
    componentAppliedLinearAccClamp=<1000 1000 1000>
    componentLinearAccScale=<1 1 1>
    componentLinearVelScale=<1 1 1>
    gravityOverride {z=-980}
    physicsBodyDefinitions [{
      boundBone {boneName=dyn_fur_arm_l_A}
      boxExtents=<15 15 5>
      constraintSetup {angularLimitsMax=<10 10 10>, angularLimitsMin=<-10 -30 -10>, angularTarget=<1 0 0>}
      localJointOffset=<1 0 0>
     }]
    relativeSpaceBone {boneName=root}
   }
   n13 = Anim Dynamics - Bone: dyn_fur_arm_l_B(ComponentPose=@n12, Alpha=1, ExternalForce="0,0,0") {
    angularDampingOverride=0.99
    angularSpringConstant=5000
    bAngularSpring=true
    bGravityOverrideInSimSpace=true
    bOverrideAngularDamping=true
    bUseGravityOverride=true
    simulationSpace=BoneRelative
    alphaBoolBlend {blendOption=Linear}
    boundBone {boneName=dyn_fur_arm_l_B}
    componentAppliedLinearAccClamp=<1000 1000 1000>
    componentLinearAccScale=<1 1 1>
    componentLinearVelScale=<1 1 1>
    gravityOverride {z=-980}
    physicsBodyDefinitions [{
      boundBone {boneName=dyn_fur_arm_l_B}
      boxExtents=<18 18 18>
      constraintSetup {angularLimitsMax=<10 10 10>, angularLimitsMin=<-10 -30 -10>, angularTarget=<1 0 0>}
      localJointOffset=<1 0 0>
     }]
    relativeSpaceBone {boneName=root}
   }
   n14 = Anim Dynamics - Bone: dyn_fur_arm_l_C(ComponentPose=@n13, Alpha=1, ExternalForce="0,0,0") {
    angularDampingOverride=0.99
    angularSpringConstant=5000
    bAngularSpring=true
    bGravityOverrideInSimSpace=true
    bOverrideAngularDamping=true
    simulationSpace=BoneRelative
    alphaBoolBlend {blendOption=Linear}
    boundBone {boneName=dyn_fur_arm_l_C}
    componentAppliedLinearAccClamp=<1000 1000 1000>
    componentLinearAccScale=<1 1 1>
    componentLinearVelScale=<1 1 1>
    gravityOverride {z=-980}
    physicsBodyDefinitions [{
      boundBone {boneName=dyn_fur_arm_l_C}
      boxExtents=<16 16 16>
      constraintSetup {angularLimitsMax=<10 10 10>, angularLimitsMin=<-10 -30 -10>, angularTarget=<1 0 0>}
      localJointOffset=<1 0 0>
     }]
    relativeSpaceBone {boneName=root}
   }
   n15 = Anim Dynamics - Bone: dyn_fur_arm_l_D(ComponentPose=@n14, Alpha=1, ExternalForce="0,0,0") {
    angularDampingOverride=0.99
    angularSpringConstant=5000
    bAngularSpring=true
    bGravityOverrideInSimSpace=true
    bOverrideAngularDamping=true
    simulationSpace=BoneRelative
    alphaBoolBlend {blendOption=Linear}
    boundBone {boneName=dyn_fur_arm_l_D}
    componentAppliedLinearAccClamp=<1000 1000 1000>
    componentLinearAccScale=<1 1 1>
    componentLinearVelScale=<1 1 1>
    gravityOverride {z=-980}
    physicsBodyDefinitions [{
      boundBone {boneName=dyn_fur_arm_l_D}
      boxExtents=<17 17 17>
      constraintSetup {angularLimitsMax=<10 10 10>, angularLimitsMin=<-10 -30 -10>, angularTarget=<1 0 0>}
      localJointOffset=<1 0 0>
     }]
    relativeSpaceBone {boneName=root}
   }
   n16 = Anim Dynamics - Bone: dyn_fur_arm_r_A(ComponentPose=@n15, Alpha=1, ExternalForce="0,0,0") {
    angularDampingOverride=0.99
    angularSpringConstant=5000
    bAngularSpring=true
    bGravityOverrideInSimSpace=true
    bOverrideAngularDamping=true
    bUseGravityOverride=true
    simulationSpace=BoneRelative
    alphaBoolBlend {blendOption=Linear}
    boundBone {boneName=dyn_fur_arm_r_A}
    componentAppliedLinearAccClamp=<1000 1000 1000>
    componentLinearAccScale=<1 1 1>
    componentLinearVelScale=<1 1 1>
    gravityOverride {z=-980}
    physicsBodyDefinitions [{
      boundBone {boneName=dyn_fur_arm_r_A}
      boxExtents=<15 15 5>
      constraintSetup {angularLimitsMax=<10 10 10>, angularLimitsMin=<-10 -30 -10>, angularTarget=<1 0 0>}
      localJointOffset=<1 0 0>
     }]
    relativeSpaceBone {boneName=root}
   }
   n17 = Anim Dynamics - Bone: dyn_fur_arm_r_B(ComponentPose=@n16, Alpha=1, ExternalForce="0,0,0") {
    angularDampingOverride=0.99
    angularSpringConstant=5000
    bAngularSpring=true
    bGravityOverrideInSimSpace=true
    bOverrideAngularDamping=true
    bUseGravityOverride=true
    simulationSpace=BoneRelative
    alphaBoolBlend {blendOption=Linear}
    boundBone {boneName=dyn_fur_arm_r_B}
    componentAppliedLinearAccClamp=<1000 1000 1000>
    componentLinearAccScale=<1 1 1>
    componentLinearVelScale=<1 1 1>
    gravityOverride {z=-980}
    physicsBodyDefinitions [{
      boundBone {boneName=dyn_fur_arm_r_B}
      boxExtents=<18 18 18>
      constraintSetup {angularLimitsMax=<10 10 10>, angularLimitsMin=<-10 -30 -10>, angularTarget=<1 0 0>}
      localJointOffset=<1 0 0>
     }]
    relativeSpaceBone {boneName=root}
   }
   n18 = Anim Dynamics - Bone: dyn_fur_arm_r_C(ComponentPose=@n17, Alpha=1, ExternalForce="0,0,0") {
    angularDampingOverride=0.99
    angularSpringConstant=5000
    bAngularSpring=true
    bGravityOverrideInSimSpace=true
    bOverrideAngularDamping=true
    simulationSpace=BoneRelative
    alphaBoolBlend {blendOption=Linear}
    boundBone {boneName=dyn_fur_arm_r_C}
    componentAppliedLinearAccClamp=<1000 1000 1000>
    componentLinearAccScale=<1 1 1>
    componentLinearVelScale=<1 1 1>
    gravityOverride {z=-980}
    physicsBodyDefinitions [{
      boundBone {boneName=dyn_fur_arm_r_C}
      boxExtents=<16 16 16>
      constraintSetup {angularLimitsMax=<10 10 10>, angularLimitsMin=<-10 -30 -10>, angularTarget=<1 0 0>}
      localJointOffset=<1 0 0>
     }]
    relativeSpaceBone {boneName=root}
   }
   n19 = Anim Dynamics - Bone: dyn_fur_arm_r_D(ComponentPose=@n18, Alpha=1, ExternalForce="0,0,0") {
    angularDampingOverride=0.99
    angularSpringConstant=5000
    bAngularSpring=true
    bGravityOverrideInSimSpace=true
    bOverrideAngularDamping=true
    simulationSpace=BoneRelative
    alphaBoolBlend {blendOption=Linear}
    boundBone {boneName=dyn_fur_arm_r_D}
    componentAppliedLinearAccClamp=<1000 1000 1000>
    componentLinearAccScale=<1 1 1>
    componentLinearVelScale=<1 1 1>
    gravityOverride {z=-980}
    physicsBodyDefinitions [{
      boundBone {boneName=dyn_fur_arm_r_D}
      boxExtents=<17 17 17>
      constraintSetup {angularLimitsMax=<10 10 10>, angularLimitsMin=<-10 -30 -10>, angularTarget=<1 0 0>}
      localJointOffset=<1 0 0>
     }]
    relativeSpaceBone {boneName=root}
   }
   n20 = Anim Dynamics - Bone: dyn_pants_l_fr_01!(SimulationSpace:AnimPhysSimSpaceType Enum=BoneRelative, ComponentPose=@n19, Alpha=1, LinearSpringConstant:float=30, ExternalForce="0,0,0") -> (Pose) {
    angularDampingOverride=0.9999
    angularSpringConstant=2000
    bAngularSpring=true
    bGravityOverrideInSimSpace=true
    bLinearSpring=true
    bOverrideAngularDamping=true
    bOverrideLinearDamping=true
    bUseGravityOverride=true
    linearDampingOverride=0.9999
    linearSpringConstant=30
    simulationSpace=BoneRelative
    alphaBoolBlend {blendOption=Linear}
    boundBone {boneName=dyn_pants_l_fr_01}
    componentAppliedLinearAccClamp=<1000 1000 1000>
    componentLinearAccScale=<1 1 1>
    componentLinearVelScale=<1 1 1>
    gravityOverride {z=-980}
    physicsBodyDefinitions [{
      boundBone {boneName=dyn_pants_l_fr_01}
      boxExtents=<7 7 7>
      constraintSetup {
       angularLimitsMax=<0.0001 0.0001 0.0001>
       angularLimitsMin=<-0.0001 -7 -0.0001>
       angularTarget=<0 0 0>
       linearAxesMax=<0.2 0 0>
       linearAxesMin=<-5 0 0>
      }
      localJointOffset=<0.1 0 0>
     }]
    relativeSpaceBone {boneName=calf_l}
    simSpaceSettings {maxAngularAcceleration=0, maxAngularVelocity=0}
   }
   n21 = Anim Dynamics - Bone: dyn_pants_l_inner_01!(SimulationSpace:AnimPhysSimSpaceType Enum=BoneRelative, ComponentPose=@n20, Alpha=1, LinearSpringConstant:float=30, ExternalForce="0,0,0") -> (Pose) {
    angularDampingOverride=0.9999
    angularSpringConstant=2000
    bAngularSpring=true
    bGravityOverrideInSimSpace=true
    bLinearSpring=true
    bOverrideAngularDamping=true
    bOverrideLinearDamping=true
    bUseGravityOverride=true
    linearDampingOverride=0.9999
    linearSpringConstant=30
    simulationSpace=BoneRelative
    alphaBoolBlend {blendOption=Linear}
    boundBone {boneName=dyn_pants_l_inner_01}
    componentAppliedLinearAccClamp=<1000 1000 1000>
    componentLinearAccScale=<1 1 1>
    componentLinearVelScale=<1 1 1>
    gravityOverride {z=-980}
    physicsBodyDefinitions [{
      boundBone {boneName=dyn_pants_l_inner_01}
      boxExtents=<7 7 7>
      constraintSetup {
       angularLimitsMax=<0.0001 0.0001 0.0001>
       angularLimitsMin=<-0.0001 -7 -0.0001>
       angularTarget=<0 0 0>
       linearAxesMax=<0.2 0 0>
       linearAxesMin=<-5 0 0>
      }
      localJointOffset=<0.1 0 0>
     }]
    relativeSpaceBone {boneName=calf_l}
    simSpaceSettings {maxAngularAcceleration=0, maxAngularVelocity=0}
   }
   n22 = Anim Dynamics - Bone: dyn_pants_l_bck_01!(SimulationSpace:AnimPhysSimSpaceType Enum=BoneRelative, ComponentPose=@n21, Alpha=1, LinearSpringConstant:float=30, ExternalForce="0,0,0") -> (Pose) {
    angularDampingOverride=0.9999
    angularSpringConstant=2000
    bAngularSpring=true
    bGravityOverrideInSimSpace=true
    bLinearSpring=true
    bOverrideAngularDamping=true
    bOverrideLinearDamping=true
    bUseGravityOverride=true
    linearDampingOverride=0.9999
    linearSpringConstant=30
    simulationSpace=BoneRelative
    alphaBoolBlend {blendOption=Linear}
    boundBone {boneName=dyn_pants_l_bck_01}
    componentAppliedLinearAccClamp=<1000 1000 1000>
    componentLinearAccScale=<1 1 1>
    componentLinearVelScale=<1 1 1>
    gravityOverride {z=-980}
    physicsBodyDefinitions [{
      boundBone {boneName=dyn_pants_l_bck_01}
      boxExtents=<7 7 7>
      constraintSetup {
       angularLimitsMax=<0.0001 0.0001 0.0001>
       angularLimitsMin=<-0.0001 -7 -0.0001>
       angularTarget=<0 0 0>
       linearAxesMax=<0.2 0 0>
       linearAxesMin=<-5 0 0>
      }
      localJointOffset=<0.1 0 0>
     }]
    relativeSpaceBone {boneName=calf_l}
    simSpaceSettings {maxAngularAcceleration=0, maxAngularVelocity=0}
   }
   n23 = Anim Dynamics - Bone: dyn_pants_l_outer_01!(SimulationSpace:AnimPhysSimSpaceType Enum=BoneRelative, ComponentPose=@n22, Alpha=1, LinearSpringConstant:float=30, ExternalForce="0,0,0") -> (Pose) {
    angularDampingOverride=0.9999
    angularSpringConstant=2000
    bAngularSpring=true
    bGravityOverrideInSimSpace=true
    bLinearSpring=true
    bOverrideAngularDamping=true
    bOverrideLinearDamping=true
    bUseGravityOverride=true
    linearDampingOverride=0.9999
    linearSpringConstant=30
    simulationSpace=BoneRelative
    alphaBoolBlend {blendOption=Linear}
    boundBone {boneName=dyn_pants_l_outer_01}
    componentAppliedLinearAccClamp=<1000 1000 1000>
    componentLinearAccScale=<1 1 1>
    componentLinearVelScale=<1 1 1>
    gravityOverride {z=-980}
    physicsBodyDefinitions [{
      boundBone {boneName=dyn_pants_l_outer_01}
      boxExtents=<7 7 7>
      constraintSetup {
       angularLimitsMax=<0.0001 0.0001 0.0001>
       angularLimitsMin=<-0.0001 -7 -0.0001>
       angularTarget=<0 0 0>
       linearAxesMax=<0.2 0 0>
       linearAxesMin=<-5 0 0>
      }
      localJointOffset=<0.1 0 0>
     }]
    relativeSpaceBone {boneName=calf_l}
    simSpaceSettings {maxAngularAcceleration=0, maxAngularVelocity=0}
   }
   n24 = Anim Dynamics - Bone: dyn_pants_l_fr_02!(SimulationSpace:AnimPhysSimSpaceType Enum=BoneRelative, ComponentPose=@n23, Alpha=1, LinearSpringConstant:float=30, ExternalForce="0,0,0") -> (Pose) {
    angularDampingOverride=0.9999
    angularSpringConstant=2000
    bAngularSpring=true
    bGravityOverrideInSimSpace=true
    bLinearSpring=true
    bOverrideAngularDamping=true
    bOverrideLinearDamping=true
    bUseGravityOverride=true
    linearDampingOverride=0.9999
    linearSpringConstant=30
    simulationSpace=BoneRelative
    alphaBoolBlend {blendOption=Linear}
    boundBone {boneName=dyn_pants_l_fr_02}
    componentAppliedLinearAccClamp=<1000 1000 1000>
    componentLinearAccScale=<1 1 1>
    componentLinearVelScale=<1 1 1>
    gravityOverride {z=-980}
    physicsBodyDefinitions [{
      boundBone {boneName=dyn_pants_l_fr_02}
      boxExtents=<7 7 7>
      constraintSetup {
       angularLimitsMax=<0.0001 0.0001 0.0001>
       angularLimitsMin=<-0.0001 -7 -0.0001>
       angularTarget=<0 0 0>
       linearAxesMax=<0.2 0 1>
       linearAxesMin=<-3 0 0>
      }
      localJointOffset=<0.1 0 0>
     }]
    relativeSpaceBone {boneName=calf_l}
    simSpaceSettings {maxAngularAcceleration=0, maxAngularVelocity=0}
   }
   n25 = Anim Dynamics - Bone: dyn_pants_l_inner_02!(SimulationSpace:AnimPhysSimSpaceType Enum=BoneRelative, ComponentPose=@n24, Alpha=1, LinearSpringConstant:float=30, ExternalForce="0,0,0") -> (Pose) {
    angularDampingOverride=0.9999
    angularSpringConstant=2000
    bAngularSpring=true
    bGravityOverrideInSimSpace=true
    bLinearSpring=true
    bOverrideAngularDamping=true
    bOverrideLinearDamping=true
    bUseGravityOverride=true
    linearDampingOverride=0.9999
    linearSpringConstant=30
    simulationSpace=BoneRelative
    alphaBoolBlend {blendOption=Linear}
    boundBone {boneName=dyn_pants_l_inner_02}
    componentAppliedLinearAccClamp=<1000 1000 1000>
    componentLinearAccScale=<1 1 1>
    componentLinearVelScale=<1 1 1>
    gravityOverride {z=-980}
    physicsBodyDefinitions [{
      boundBone {boneName=dyn_pants_l_inner_02}
      boxExtents=<7 7 7>
      constraintSetup {
       angularLimitsMax=<0.0001 0.0001 0.0001>
       angularLimitsMin=<-0.0001 -7 -0.0001>
       angularTarget=<0 0 0>
       linearAxesMax=<0.2 0 1>
       linearAxesMin=<-3 0 0>
      }
      localJointOffset=<0.1 0 0>
     }]
    relativeSpaceBone {boneName=calf_l}
    simSpaceSettings {maxAngularAcceleration=0, maxAngularVelocity=0}
   }
   n26 = Anim Dynamics - Bone: dyn_pants_l_bck_02!(SimulationSpace:AnimPhysSimSpaceType Enum=BoneRelative, ComponentPose=@n25, Alpha=1, LinearSpringConstant:float=30, ExternalForce="0,0,0") -> (Pose) {
    angularDampingOverride=0.9999
    angularSpringConstant=2000
    bAngularSpring=true
    bGravityOverrideInSimSpace=true
    bLinearSpring=true
    bOverrideAngularDamping=true
    bOverrideLinearDamping=true
    bUseGravityOverride=true
    linearDampingOverride=0.9999
    linearSpringConstant=30
    simulationSpace=BoneRelative
    alphaBoolBlend {blendOption=Linear}
    boundBone {boneName=dyn_pants_l_bck_02}
    componentAppliedLinearAccClamp=<1000 1000 1000>
    componentLinearAccScale=<1 1 1>
    componentLinearVelScale=<1 1 1>
    gravityOverride {z=-980}
    physicsBodyDefinitions [{
      boundBone {boneName=dyn_pants_l_bck_02}
      boxExtents=<7 7 7>
      constraintSetup {
       angularLimitsMax=<0.0001 0.0001 0.0001>
       angularLimitsMin=<-0.0001 -7 -0.0001>
       angularTarget=<0 0 0>
       linearAxesMax=<0.2 0 1>
       linearAxesMin=<-3 0 0>
      }
      localJointOffset=<0.1 0 0>
     }]
    relativeSpaceBone {boneName=calf_l}
    simSpaceSettings {maxAngularAcceleration=0, maxAngularVelocity=0}
   }
   n27 = Anim Dynamics - Bone: dyn_pants_l_outer_02!(SimulationSpace:AnimPhysSimSpaceType Enum=BoneRelative, ComponentPose=@n26, Alpha=1, LinearSpringConstant:float=30, ExternalForce="0,0,0") -> (Pose) {
    angularDampingOverride=0.9999
    angularSpringConstant=2000
    bAngularSpring=true
    bGravityOverrideInSimSpace=true
    bLinearSpring=true
    bOverrideAngularDamping=true
    bOverrideLinearDamping=true
    bUseGravityOverride=true
    linearDampingOverride=0.9999
    linearSpringConstant=30
    simulationSpace=BoneRelative
    alphaBoolBlend {blendOption=Linear}
    boundBone {boneName=dyn_pants_l_outer_02}
    componentAppliedLinearAccClamp=<1000 1000 1000>
    componentLinearAccScale=<1 1 1>
    componentLinearVelScale=<1 1 1>
    gravityOverride {z=-980}
    physicsBodyDefinitions [{
      boundBone {boneName=dyn_pants_l_outer_02}
      boxExtents=<7 7 7>
      constraintSetup {
       angularLimitsMax=<0.0001 0.0001 0.0001>
       angularLimitsMin=<-0.0001 -7 -0.0001>
       angularTarget=<0 0 0>
       linearAxesMax=<0.2 0 1>
       linearAxesMin=<-3 0 0>
      }
      localJointOffset=<0.1 0 0>
     }]
    relativeSpaceBone {boneName=calf_l}
    simSpaceSettings {maxAngularAcceleration=0, maxAngularVelocity=0}
   }
   n28 = Anim Dynamics - Bone: dyn_pants_r_fr_01!(SimulationSpace:AnimPhysSimSpaceType Enum=BoneRelative, ComponentPose=@n27, Alpha=1, LinearSpringConstant:float=30, ExternalForce="0,0,0") -> (Pose) {
    angularDampingOverride=0.9999
    angularSpringConstant=2000
    bAngularSpring=true
    bGravityOverrideInSimSpace=true
    bLinearSpring=true
    bOverrideAngularDamping=true
    bOverrideLinearDamping=true
    bUseGravityOverride=true
    linearDampingOverride=0.9999
    linearSpringConstant=30
    simulationSpace=BoneRelative
    alphaBoolBlend {blendOption=Linear}
    boundBone {boneName=dyn_pants_r_fr_01}
    componentAppliedLinearAccClamp=<1000 1000 1000>
    componentLinearAccScale=<1 1 1>
    componentLinearVelScale=<1 1 1>
    gravityOverride {z=-980}
    physicsBodyDefinitions [{
      boundBone {boneName=dyn_pants_r_fr_01}
      boxExtents=<7 7 7>
      constraintSetup {
       angularLimitsMax=<0.0001 0.0001 0.0001>
       angularLimitsMin=<-0.0001 -7 -0.0001>
       angularTarget=<0 0 0>
       linearAxesMax=<0.2 0 0>
       linearAxesMin=<-5 0 0>
      }
      localJointOffset=<0.1 0 0>
     }]
    relativeSpaceBone {boneName=calf_r}
    simSpaceSettings {maxAngularAcceleration=0, maxAngularVelocity=0}
   }
   n29 = Anim Dynamics - Bone: dyn_pants_r_inner_01!(SimulationSpace:AnimPhysSimSpaceType Enum=BoneRelative, ComponentPose=@n28, Alpha=1, LinearSpringConstant:float=30, ExternalForce="0,0,0") -> (Pose) {
    angularDampingOverride=0.9999
    angularSpringConstant=2000
    bAngularSpring=true
    bGravityOverrideInSimSpace=true
    bLinearSpring=true
    bOverrideAngularDamping=true
    bOverrideLinearDamping=true
    bUseGravityOverride=true
    linearDampingOverride=0.9999
    linearSpringConstant=30
    simulationSpace=BoneRelative
    alphaBoolBlend {blendOption=Linear}
    boundBone {boneName=dyn_pants_r_inner_01}
    componentAppliedLinearAccClamp=<1000 1000 1000>
    componentLinearAccScale=<1 1 1>
    componentLinearVelScale=<1 1 1>
    gravityOverride {z=-980}
    physicsBodyDefinitions [{
      boundBone {boneName=dyn_pants_r_inner_01}
      boxExtents=<7 7 7>
      constraintSetup {
       angularLimitsMax=<0.0001 0.0001 0.0001>
       angularLimitsMin=<-0.0001 -7 -0.0001>
       angularTarget=<0 0 0>
       linearAxesMax=<0.2 0 0>
       linearAxesMin=<-5 0 0>
      }
      localJointOffset=<0.1 0 0>
     }]
    relativeSpaceBone {boneName=calf_r}
    simSpaceSettings {maxAngularAcceleration=0, maxAngularVelocity=0}
   }
   n30 = Anim Dynamics - Bone: dyn_pants_r_bck_01!(SimulationSpace:AnimPhysSimSpaceType Enum=BoneRelative, ComponentPose=@n29, Alpha=1, LinearSpringConstant:float=30, ExternalForce="0,0,0") -> (Pose) {
    angularDampingOverride=0.9999
    angularSpringConstant=2000
    bAngularSpring=true
    bGravityOverrideInSimSpace=true
    bLinearSpring=true
    bOverrideAngularDamping=true
    bOverrideLinearDamping=true
    bUseGravityOverride=true
    linearDampingOverride=0.9999
    linearSpringConstant=30
    simulationSpace=BoneRelative
    alphaBoolBlend {blendOption=Linear}
    boundBone {boneName=dyn_pants_r_bck_01}
    componentAppliedLinearAccClamp=<1000 1000 1000>
    componentLinearAccScale=<1 1 1>
    componentLinearVelScale=<1 1 1>
    gravityOverride {z=-980}
    physicsBodyDefinitions [{
      boundBone {boneName=dyn_pants_r_bck_01}
      boxExtents=<7 7 7>
      constraintSetup {
       angularLimitsMax=<0.0001 0.0001 0.0001>
       angularLimitsMin=<-0.0001 -7 -0.0001>
       angularTarget=<0 0 0>
       linearAxesMax=<0.2 0 0>
       linearAxesMin=<-5 0 0>
      }
      localJointOffset=<0.1 0 0>
     }]
    relativeSpaceBone {boneName=calf_r}
    simSpaceSettings {maxAngularAcceleration=0, maxAngularVelocity=0}
   }
   n31 = Anim Dynamics - Bone: dyn_pants_r_outer_01!(SimulationSpace:AnimPhysSimSpaceType Enum=BoneRelative, ComponentPose=@n30, Alpha=1, LinearSpringConstant:float=30, ExternalForce="0,0,0") -> (Pose) {
    angularDampingOverride=0.9999
    angularSpringConstant=2000
    bAngularSpring=true
    bGravityOverrideInSimSpace=true
    bLinearSpring=true
    bOverrideAngularDamping=true
    bOverrideLinearDamping=true
    bUseGravityOverride=true
    linearDampingOverride=0.9999
    linearSpringConstant=30
    simulationSpace=BoneRelative
    alphaBoolBlend {blendOption=Linear}
    boundBone {boneName=dyn_pants_r_outer_01}
    componentAppliedLinearAccClamp=<1000 1000 1000>
    componentLinearAccScale=<1 1 1>
    componentLinearVelScale=<1 1 1>
    gravityOverride {z=-980}
    physicsBodyDefinitions [{
      boundBone {boneName=dyn_pants_r_outer_01}
      boxExtents=<7 7 7>
      constraintSetup {
       angularLimitsMax=<0.0001 0.0001 0.0001>
       angularLimitsMin=<-0.0001 -7 -0.0001>
       angularTarget=<0 0 0>
       linearAxesMax=<0.2 0 0>
       linearAxesMin=<-5 0 0>
      }
      localJointOffset=<0.1 0 0>
     }]
    relativeSpaceBone {boneName=calf_r}
    simSpaceSettings {maxAngularAcceleration=0, maxAngularVelocity=0}
   }
   n32 = Anim Dynamics - Bone: dyn_pants_r_fr_02!(SimulationSpace:AnimPhysSimSpaceType Enum=Component, ComponentPose=@n31, Alpha=1, LinearSpringConstant:float=30, ExternalForce="0,0,0") -> (Pose) {
    angularDampingOverride=0.9999
    angularSpringConstant=2000
    bAngularSpring=true
    bGravityOverrideInSimSpace=true
    bLinearSpring=true
    bOverrideAngularDamping=true
    bOverrideLinearDamping=true
    bUseGravityOverride=true
    linearDampingOverride=0.9999
    linearSpringConstant=30
    simulationSpace=BoneRelative
    alphaBoolBlend {blendOption=Linear}
    boundBone {boneName=dyn_pants_r_fr_02}
    componentAppliedLinearAccClamp=<1000 1000 1000>
    componentLinearAccScale=<1 1 1>
    componentLinearVelScale=<1 1 1>
    gravityOverride {z=-980}
    physicsBodyDefinitions [{
      boundBone {boneName=dyn_pants_r_fr_02}
      boxExtents=<7 7 7>
      constraintSetup {
       angularLimitsMax=<0.0001 0.0001 0.0001>
       angularLimitsMin=<-0.0001 -7 -0.0001>
       angularTarget=<0 0 0>
       linearAxesMax=<0.2 0 1>
       linearAxesMin=<-3 0 0>
      }
      localJointOffset=<0.1 0 0>
     }]
    relativeSpaceBone {boneName=calf_r}
    simSpaceSettings {maxAngularAcceleration=0, maxAngularVelocity=0}
   }
   n33 = Anim Dynamics - Bone: dyn_pants_r_inner_02!(SimulationSpace:AnimPhysSimSpaceType Enum=BoneRelative, ComponentPose=@n32, Alpha=1, LinearSpringConstant:float=30, ExternalForce="0,0,0") -> (Pose) {
    angularDampingOverride=0.9999
    angularSpringConstant=2000
    bAngularSpring=true
    bGravityOverrideInSimSpace=true
    bLinearSpring=true
    bOverrideAngularDamping=true
    bOverrideLinearDamping=true
    bUseGravityOverride=true
    linearDampingOverride=0.9999
    linearSpringConstant=30
    simulationSpace=BoneRelative
    alphaBoolBlend {blendOption=Linear}
    boundBone {boneName=dyn_pants_r_inner_02}
    componentAppliedLinearAccClamp=<1000 1000 1000>
    componentLinearAccScale=<1 1 1>
    componentLinearVelScale=<1 1 1>
    gravityOverride {z=-980}
    physicsBodyDefinitions [{
      boundBone {boneName=dyn_pants_r_inner_02}
      boxExtents=<7 7 7>
      constraintSetup {
       angularLimitsMax=<0.0001 0.0001 0.0001>
       angularLimitsMin=<-0.0001 -7 -0.0001>
       angularTarget=<0 0 0>
       linearAxesMax=<0.2 0 1>
       linearAxesMin=<-3 0 0>
      }
      localJointOffset=<0.1 0 0>
     }]
    relativeSpaceBone {boneName=calf_r}
    simSpaceSettings {maxAngularAcceleration=0, maxAngularVelocity=0}
   }
   n34 = Anim Dynamics - Bone: dyn_pants_r_bck_02!(SimulationSpace:AnimPhysSimSpaceType Enum=BoneRelative, ComponentPose=@n33, Alpha=1, LinearSpringConstant:float=30, ExternalForce="0,0,0") -> (Pose) {
    angularDampingOverride=0.9999
    angularSpringConstant=2000
    bAngularSpring=true
    bGravityOverrideInSimSpace=true
    bLinearSpring=true
    bOverrideAngularDamping=true
    bOverrideLinearDamping=true
    bUseGravityOverride=true
    linearDampingOverride=0.9999
    linearSpringConstant=30
    simulationSpace=BoneRelative
    alphaBoolBlend {blendOption=Linear}
    boundBone {boneName=dyn_pants_r_bck_02}
    componentAppliedLinearAccClamp=<1000 1000 1000>
    componentLinearAccScale=<1 1 1>
    componentLinearVelScale=<1 1 1>
    gravityOverride {z=-980}
    physicsBodyDefinitions [{
      boundBone {boneName=dyn_pants_r_bck_02}
      boxExtents=<7 7 7>
      constraintSetup {
       angularLimitsMax=<0.0001 0.0001 0.0001>
       angularLimitsMin=<-0.0001 -7 -0.0001>
       angularTarget=<0 0 0>
       linearAxesMax=<0.2 0 1>
       linearAxesMin=<-3 0 0>
      }
      localJointOffset=<0.1 0 0>
     }]
    relativeSpaceBone {boneName=calf_r}
    simSpaceSettings {maxAngularAcceleration=0, maxAngularVelocity=0}
   }
   n35 = Anim Dynamics - Bone: dyn_pants_r_outer_02!(SimulationSpace:AnimPhysSimSpaceType Enum=BoneRelative, ComponentPose=@n34, Alpha=1, LinearSpringConstant:float=30, ExternalForce="0,0,0") -> (Pose) {
    angularDampingOverride=0.9999
    angularSpringConstant=2000
    bAngularSpring=true
    bGravityOverrideInSimSpace=true
    bLinearSpring=true
    bOverrideAngularDamping=true
    bOverrideLinearDamping=true
    bUseGravityOverride=true
    linearDampingOverride=0.9999
    linearSpringConstant=30
    simulationSpace=BoneRelative
    alphaBoolBlend {blendOption=Linear}
    boundBone {boneName=dyn_pants_r_outer_02}
    componentAppliedLinearAccClamp=<1000 1000 1000>
    componentLinearAccScale=<1 1 1>
    componentLinearVelScale=<1 1 1>
    gravityOverride {z=-980}
    physicsBodyDefinitions [{
      boundBone {boneName=dyn_pants_r_outer_02}
      boxExtents=<7 7 7>
      constraintSetup {
       angularLimitsMax=<0.0001 0.0001 0.0001>
       angularLimitsMin=<-0.0001 -7 -0.0001>
       angularTarget=<0 0 0>
       linearAxesMax=<0.2 0 1>
       linearAxesMin=<-3 0 0>
      }
      localJointOffset=<0.1 0 0>
     }]
    relativeSpaceBone {boneName=calf_r}
    simSpaceSettings {maxAngularAcceleration=0, maxAngularVelocity=0}
   }
   n36 = "Constraint - dyn_pants_r_bck_01 (P)"(ComponentPose=@n35, Alpha=1, ConstraintWeights_0=1) {
    alphaBoolBlend {blendOption=Linear}
    alphaScaleBiasClamp {bInterpResult=true}
    boneToModify {boneName=dyn_pants_r_bck_01}
    constraintSetup [{targetBone {boneName=foot_r}}]
    constraintWeights [1]
   }
   n37 = "Constraint - dyn_pants_l_bck_01 (P)"(ComponentPose=@n36, Alpha=1, ConstraintWeights_0=1) {
    alphaBoolBlend {blendOption=Linear}
    alphaScaleBiasClamp {bInterpResult=true}
    boneToModify {boneName=dyn_pants_l_bck_01}
    constraintSetup [{targetBone {boneName=foot_l}}]
    constraintWeights [1]
   }
   n38 = RigidBody(
    ComponentPose=@n37
    OverrideWorldGravity="0,0,-980"
    SimSpaceSettings=(WorldAlpha=0.25,VelocityScaleZ=1,DampingAlpha=1,MaxLinearVelocity=800,MaxAngularVelocity=8,MaxLinearAcceleration=8000,MaxAngularAcceleration=8000,ExternalLinearDragV=(X=0,Y=0,Z=0),ExternalLinearVelocity=(X=0,Y=0,Z=0),ExternalAngularVelocity=(X=0,Y=0,Z=0))
   ) {
    bClampLinearTranslationLimitToRefPose=true
    bOverrideWorldGravity=true
    overridePhysicsAsset=&/F_MED_WoolExpo_FE_Physics
    simulationSpace=BaseBoneSpace
    alphaBoolBlend {blendOption=Linear}
    alphaScaleBiasClamp {bInterpResult=true}
    baseBoneRef {boneName=pelvis}
    overrideWorldGravity {z=-980}
    simSpaceSettings {maxAngularAcceleration=8000, maxAngularVelocity=8, maxLinearAcceleration=8000, maxLinearVelocity=800, worldAlpha=0.25}
   }
   Output Pose(Result=Component To Local(ComponentPose=@n38)) {layerGroup=None}
  ]}
 EventGraph {nodes [
   n1 = Event BlueprintUpdateAnimation()
   > n2 = "Gravity Override State Function (Parameter Asset) (Nativized)"(
    CharacterPartAnimInstance=Self-Reference()
    GameTimeInSeconds=@n1.DeltaTimeX
    DebugJointName=neck_02
    Parameters=&/F_Med_WoolExpo_FaceAcc_Parameters
   )
   > Set Gravity Override(Gravity Override=@n2)
   > Set Sim Space Settings(Sim Space Settings:Sim Space Settings Structure=Break Fort Rigid Body Anim Node Parameters!(FortRigidBodyAnimNodeParameters:"Fort Rigid Body Anim Node Parameters Structure (by ref)"=@n2.OutRigidBodyAnimNodeParameters) -> (ComponentLinearAccScale:Vector=~, ComponentLinearVelScale:Vector=~, SimSpaceSettings:Sim Space Settings Structure) {is=title K2Node_BreakStruct}.SimSpaceSettings) -> (Output_Get:Sim Space Settings Structure)
   Event OnBeginPlayingEmote() {is=title K2Node_Event}
   > Set Anim Dyn Sim Space(Anim Dyn Sim Space:AnimPhysSimSpaceType Enum=BoneRelative) -> (Output_Get:AnimPhysSimSpaceType Enum)
   > n3 = Branch(Condition=Get bIsInFrontEnd())
   > "Set FE Alpha (No Emote)"("FE Alpha (No Emote)":double=0) -> (Output_Get:double)
   > Set RBAN Alpha(RBAN Alpha:double=1) -> (Output_Get:double)
   "Set FE Alpha (No Emote)"(execute=@n3.else, "FE Alpha (No Emote)"=0)
   > Set RBAN Alpha(RBAN Alpha=1)
   Event OnEndPlayingEmote() {is=title K2Node_Event}
   > Set Anim Dyn Sim Space()
   > n4 = Branch(Condition=Get bIsInFrontEnd())
   > "Set FE Alpha (No Emote)"("FE Alpha (No Emote)"=1)
   > Set RBAN Alpha(RBAN Alpha=0)
   "Set FE Alpha (No Emote)"(execute=@n4.else, "FE Alpha (No Emote)"=0)
   > Set RBAN Alpha(RBAN Alpha=1)
   Event OnCharacterCustomizationLoaded()
   > n5 = Branch(Condition=Get bIsInFrontEnd())
   > "Set FE Alpha (No Emote)"("FE Alpha (No Emote)"=1)
   > Set RBAN Alpha(RBAN Alpha=0)
   "Set FE Alpha (No Emote)"(execute=@n5.else, "FE Alpha (No Emote)"=0)
   > Set RBAN Alpha(RBAN Alpha=1)
   Event OnBeginSkydiving()
   > Set PantSpringConstant(PantSpringConstant:double=1000) -> (Output_Get:double)
   Event OnEndSkydiving()
   > Set PantSpringConstant(PantSpringConstant=30)
  ]}
]
variables [
 Anim Dyn Sim Space {type=AnimPhysSimSpaceType, default=RootRelative}
 Body Mesh {type=SkeletalMeshComponent}
 "FE Alpha (No Emote)" {type=number}
 Gravity Override {type=Vector}
 PantSpringConstant {type=number, default=30}
 RBAN Alpha {type=number}
 Sim Space Settings {
  type=SimSpaceSettings
  default {
   dampingAlpha=1
   maxAngularAcceleration=10000
   maxAngularVelocity=1000
   maxLinearAcceleration=10000
   maxLinearVelocity=1000
   velocityScaleZ=1
   worldAlpha=0.2
   externalLinearDragV=<0 0 0>
  }
 }
 Skel_Mesh {type=SkeletalMeshComponent}
]
