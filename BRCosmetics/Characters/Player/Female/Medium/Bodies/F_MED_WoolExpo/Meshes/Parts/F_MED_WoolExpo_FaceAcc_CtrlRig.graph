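# Control Rig for the F_MED_WoolExpo character's face accessory part that applies corrective offsets to dynamic hair bones, including the right front and side strands, as the head moves.
# A spherical pose reader on the head drives HeadUp.
# Head rotation and translation relative to spine_04 are remapped and multiplied into Crouch Move Head Up.
# Each value weights its own Modify Transforms pass through an AlphaInterp with interpolation switched off.
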ControlRigBlueprint /BRCosmetics/Characters/Player/Female/Medium/Bodies/F_MED_WoolExpo/Meshes/Parts/F_MED_WoolExpo_FaceAcc_CtrlRig
graphs [
 RigVMFunctionLibrary {links [], nodes []}
 RigVMFunction_Sequence_1_ContainedGraph {nodes [
   n1 = RigVMFunctionEntryNode()
   RigVMFunction_Sequence = &RigVMFunction_Sequence(ExecuteContext=@n1.ExecuteContext)
   n2 = &RigVMFunction_Sequence(ExecuteContext=@RigVMFunction_Sequence.B)
   n3 = &RigVMFunction_Sequence(ExecuteContext=@n2.B)
   RigVMFunctionReturnNode(A=@RigVMFunction_Sequence.A, B=@n2.A, C=@n3.A, D=@n3.B)
  ]}
 RigVMModel {nodes [
   RigUnit_BeginExecution = &RigUnit_BeginExecution()
   RigVMFunction_Sequence = RigVMAggregateNode(ExecuteContext=@RigUnit_BeginExecution.ExecutePin)
   n1 = &RigUnit_SphericalPoseReader(
    ActiveRegionScaleFactors=(PositiveWidth=1,NegativeWidth=1,PositiveHeight=1,NegativeHeight=1)
    ActiveRegionSize=0.2
    Debug=(bDrawDebug=false,bDraw2D=False,bDrawLocalAxes=False,DebugScale=25,DebugSegments=20,DebugThickness=0.25)
    DriverAxis=(X=1,Y=0,Z=0)
    DriverItem=(Type=Bone,Name="head")
    FalloffRegionScaleFactors=(PositiveWidth=1.5,NegativeWidth=1,PositiveHeight=1,NegativeHeight=1)
    FalloffSize=0.32
    FlipHeightScaling=False
    FlipWidthScaling=False
    OptionalParentItem=(Type=Bone,Name="None")
    RotationOffset=(X=-64.889,Y=15.1763,Z=36.1456)
    ExecutePin=@RigVMFunction_Sequence.A
   )
   RigVMVariableNode(ExecuteContext=@RigVMFunction_Sequence.B, Value=@n1.OutputParam) {variable=HeadUp}
   n2 = &RigUnit_GetRelativeTransformForItem(Child=(Type=Bone,Name="head"), Parent=(Type=Bone,Name="spine_04"), bChildInitial=False, bParentInitial=false)
   n3 = RigVMDispatchNode(Struct=@n2.RelativeTransform.Translation) {template="DISPATCH_RigVMDispatch_BreakStruct(in Struct,out Elements)"}
   Absolute = &RigVMFunction_MathDoubleAbs(Value=&RigVMFunction_MathQuaternionToEuler(RotationOrder=ZYX, Value=@n2.RelativeTransform.Rotation).Result.Z)
   > RigVMVariableNode(Value=&RigVMFunction_MathDoubleMul(
     B=&RigVMFunction_MathDoubleRemap(SourceMaximum=4, SourceMinimum=1, TargetMaximum=1, TargetMinimum=0, Value=@n3.Elements.Z).Result
     A=&RigVMFunction_MathDoubleRemap(SourceMaximum=80, SourceMinimum=60, TargetMaximum=1, TargetMinimum=0, Value=@Absolute.Result).Result
    ).Result) {variable=Crouch Move Head Up}
   AlphaInterp = &RigVMFunction_AlphaInterp(
    InterpSpeedDecreasing=20
    InterpSpeedIncreasing=20
    Scale=1
    bClampResult=False
    bInterpResult=False
    Value=RigVMVariableNode() {variable=HeadUp}.Value
   )
   &RigUnit_ModifyTransforms(
    ItemToModify="((Item=(Type=Bone,Name=\"dyn_hair_fr_r_01\"),Transform=(Rotation=(X=0,Y=0,Z=0,W=1),Translation=(X=0,Y=0,Z=0),Scale3D=(X=1,Y=1,Z=1))),(Item=(Type=Bone,Name=\"dyn_hair_side_r_01\"),Transform=(Rotation=(X=0,Y=0,Z=0,W=1),Translation=(X=0,Y=0,Z=-1.2317),Scale3D=(X= ...(truncated)"
    WeightMinimum=0
    Weight=@AlphaInterp.Result
    ExecutePin=@RigVMFunction_Sequence.C
   )
   n4 = &RigVMFunction_AlphaInterp(
    InterpSpeedDecreasing=20
    InterpSpeedIncreasing=20
    Scale=1
    bClampResult=False
    bInterpResult=False
    Value=RigVMVariableNode() {variable=Crouch Move Head Up}.Value
   )
   > &RigUnit_ModifyTransforms(
    ItemToModify="((Item=(Type=Bone,Name=\"dyn_hair_side_r_01\"),Transform=(Rotation=(X=0.0151,Y=0.173,Z=-0.0858,W=0.9811),Translation=(X=-0.3769,Y=0.0665,Z=-2.1718),Scale3D=(X=1,Y=1,Z=1))),(Item=(Type=Bone,Name=\"dyn_hair_fr_r_01\"),Transform=(Rotation=(X=0.0076,Y=0.0868,Z=-0. ...(truncated)"
    WeightMinimum=0
    Weight=@n4.Result
   )
  ]}
]
hierarchy [
 root {loc=<0 0 0>, rot=<0 0 0 1>}
 pelvis {parent=root, loc=<0 -0.5301 98.0571>, rot=<0.0127 0.707 -0.0127 -0.707>}
 spine_01 {parent=pelvis, loc=<6.1208 -0.9328 0>, rot=<0 0 0.034 -0.9994>}
 spine_02 {parent=spine_01, loc=<6.5997 -0.2095 0>, rot=<0 0 0.0238 -0.9997>}
 spine_03 {parent=spine_02, loc=<6.4914 -0.1237 0>, rot=<0 0 0.0149 -0.9999>}
 spine_04 {parent=spine_03, loc=<6.9598 0.0794 0>, rot=<0 0 -0.0132 -0.9999>}
 spine_05 {parent=spine_04, loc=<6.9987 0 0>, rot=<0 0 -0.0015 -1>}
 neck_01 {parent=spine_05, loc=<11.3138 0.1529 0>, rot=<0 0 -0.2003 -0.9797>}
 neck_02 {parent=neck_01, loc=<4.5627 0 0>, rot=<0 0 0.0504 -0.9987>}
 head {parent=neck_02, loc=<5.3914 0 0>, rot=<0 0 0.115 -0.9934>}
 dyn_ear_l_01 {parent=head, loc=<10.4135 0.938 -4.412>, rot=<0.0183 0.1261 -0.114 -0.9853>}
 dyn_ear_l_02 {parent=dyn_ear_l_01, loc=<4.1532 0 0>, rot=<-0.0396 -0.1363 -0.0598 -0.9881>}
 dyn_ear_l_03 {parent=dyn_ear_l_02, loc=<3.214 0 0>, rot=<0.0436 0 0 -0.999>}
 dyn_earFluff_l_A_01 {parent=head, loc=<9.8663 0.5533 -6.049>, rot=<0.1334 -0.2476 -0.0133 -0.9595>}
 dyn_earFluff_l_A_02 {parent=dyn_earFluff_l_A_01, loc=<2.3191 0 0>, rot=<-0.0001 0 0.0436 -0.999>}
 dyn_earFluff_l_B {parent=head, loc=<8.7963 2.6073 -6.158>, rot=<-0.0322 -0.7366 -0.0295 -0.6749>}
 dyn_earFluff_l_C {parent=head, loc=<8.822 -0.5108 -7.516>, rot=<0.0714 -0.816 0.05 -0.5714>}
 dyn_earFluff_l_D {parent=head, loc=<7.6901 1.0909 -7.261>, rot=<0.0395 -0.9054 0.0184 -0.4222>}
 dyn_earFluff_l_E {parent=head, loc=<7.3747 -2.1227 -7.96>, rot=<-0.0402 -0.8836 0.0773 -0.46>}
 dyn_ear_r_01 {parent=head, loc=<10.4135 0.938 4.4118>, rot=<0.9782 0.1305 -0.109 0.1189>}
 dyn_ear_r_02 {parent=dyn_ear_r_01, loc=<4.1531 0 0>, rot=<0 -0.1465 0.0258 -0.9889>}
 dyn_ear_r_03 {parent=dyn_ear_r_02, loc=<3.214 0 0>, rot=<-0.0872 0 0 -0.9962>}
 dyn_earFluff_r_A_01 {parent=head, loc=<9.8663 0.5533 6.0489>, rot=<0.9596 0.0133 0.2476 -0.1334>}
 dyn_earFluff_r_A_02 {parent=dyn_earFluff_r_A_01, loc=<2.319 0 0>, rot=<-0.0518 -0.2093 0.0043 -0.9765>}
 dyn_earFluff_r_B {parent=head, loc=<8.822 -0.5108 7.5161>, rot=<0.7071 0 0.7071 0>}
 dyn_earFluff_r_C {parent=head, loc=<8.7963 2.6073 6.1581>, rot=<0.6391 0.0997 0.7592 0.072>}
 dyn_earFluff_r_D {parent=head, loc=<7.3747 -2.1227 7.9605>, rot=<0.5736 0 0.8192 0>}
 dyn_earFluff_r_E {parent=head, loc=<7.6901 1.0909 7.2606>, rot=<0.4263 0.042 0.9026 -0.042>}
 dyn_bangs_A {parent=head, loc=<7.112 7.9954 4.9465>, rot=<0.0967 0.3368 0.9365 -0.0111>}
 dyn_bangs_B {parent=head, loc=<7.0957 9.9587 2.5013>, rot=<0 0.5736 0.8192 0>}
 dyn_bangs_C {parent=head, loc=<7.091 10.536 -0.4941>, rot=<0.028 -0.7653 -0.6422 0.0334>}
 dyn_bangs_D {parent=head, loc=<7.1947 9.3272 -3.9583>, rot=<0 -0.866 -0.5 0>}
 dyn_bangs_E {parent=head, loc=<7.3875 7.6984 -5.6771>, rot=<0.0831 -0.9501 -0.2996 -0.0262>}
 dyn_hair_fr_r_01 {parent=head, loc=<2.4412 2.5648 7.6623>, rot=<0.0934 0.3061 0.9391 -0.125>}
 dyn_hair_fr_r_02 {parent=dyn_hair_fr_r_01, loc=<3.447 0 0>, rot=<0.0263 -0.0552 0.1797 -0.9818>}
 dyn_hair_fr_r_03 {parent=dyn_hair_fr_r_02, loc=<3.5604 0 0>, rot=<0.0985 0.013 0.1299 -0.9865>}
 dyn_hair_side_r_01 {parent=head, loc=<3.6163 -2.7313 7.6623>, rot=<0.1499 -0.0855 0.9849 0.009>}
 dyn_hair_side_r_02 {parent=dyn_hair_side_r_01, loc=<4.2166 0 0>, rot=<-0.0618 -0.0844 0.0742 -0.9917>}
 dyn_hair_side_r_03 {parent=dyn_hair_side_r_02, loc=<4.2972 0 0>, rot=<-0.0444 -0.0058 0.1304 -0.9905>}
 dyn_hair_bck_r_01 {parent=head, loc=<7.8493 -5.1216 4.6091>, rot=<0.1739 -0.4084 0.8948 -0.0485>}
 dyn_hair_bck_r_02 {parent=dyn_hair_bck_r_01, loc=<4.4979 0 0>, rot=<-0.044 -0.1127 -0.0087 -0.9926>}
 dyn_hair_bck_r_03 {parent=dyn_hair_bck_r_02, loc=<4.201 0 0>, rot=<0.1721 -0.1225 0.0124 -0.9774>}
 dyn_hair_bck_r_04 {parent=dyn_hair_bck_r_03, loc=<3.8303 0 0>, rot=<-0.0366 0 0 -0.9993>}
 dyn_hair_bck_c_01 {parent=head, loc=<7.86 -6.4137 0.9178>, rot=<0.1863 -0.6293 0.7373 -0.1599>}
 dyn_hair_bck_c_02 {parent=dyn_hair_bck_c_01, loc=<4.5545 0 0>, rot=<-0.0247 -0.1971 0.0248 -0.9798>}
 dyn_hair_bck_c_03 {parent=dyn_hair_bck_c_02, loc=<4.0932 0 0>, rot=<0.2015 -0.1109 -0.1034 -0.9677>}
 dyn_hair_bck_c_04 {parent=dyn_hair_bck_c_03, loc=<3.8052 0 0>, rot=<-0.2144 0 0 -0.9767>}
 dyn_hair_bck_l_01 {parent=head, loc=<7.8574 -6.1013 -2.4223>, rot=<0.0749 -0.7766 0.6106 -0.1363>}
 dyn_hair_bck_l_02 {parent=dyn_hair_bck_l_01, loc=<4.4166 0 0>, rot=<-0.0002 -0.0369 0.0066 -0.9993>}
 dyn_hair_bck_l_03 {parent=dyn_hair_bck_l_02, loc=<4.9471 0 0>, rot=<0.0645 -0.1486 -0.1254 -0.9788>}
 dyn_hair_bck_l_04 {parent=dyn_hair_bck_l_03, loc=<3.7461 0 0>, rot=<-0.0343 0 0 -0.9994>}
 dyn_hair_side_l_01 {parent=head, loc=<3.4773 -5.9224 -6.9495>, rot=<-0.1078 -0.8998 0.4154 -0.0792>}
 dyn_hair_side_l_02 {parent=dyn_hair_side_l_01, loc=<4.2926 0 0>, rot=<-0.0211 -0.0698 -0.0041 -0.9973>}
 dyn_hair_side_l_03 {parent=dyn_hair_side_l_02, loc=<4.7496 0 0>, rot=<0.1549 0 0 -0.9879>}
 dyn_hair_fr_l_01 {parent=head, loc=<2.461 0.1667 -8.4197>, rot=<0.0417 -0.9942 0.0235 -0.096>}
 dyn_hair_fr_l_02 {parent=dyn_hair_fr_l_01, loc=<4.1509 0 0>, rot=<0.0539 -0.1011 -0.1543 -0.9814>}
 dyn_hair_fr_l_03 {parent=dyn_hair_fr_l_02, loc=<3.6538 0 0>, rot=<-0.0085 -0.1011 0.0195 -0.9947>}
 clavicle_l {parent=spine_05, loc=<5.4831 -0.1053 -3.2946>, rot=<0.0465 -0.7289 -0.009 -0.6829>}
 upperarm_l {parent=clavicle_l, loc=<9.8658 -1.2136 -0.4105>, rot=<0.0164 -0.3245 -0.0269 -0.9454>}
 clavicle_r {parent=spine_05, loc=<5.4831 -0.1053 3.2946>, rot=<0.7289 0.0465 -0.6829 0.009>}
 upperarm_r {parent=clavicle_r, loc=<-9.8658 1.2136 0.4105>, rot=<0.0164 -0.3245 -0.0269 -0.9454>}
]
variables [
 AlphaRandomizer {type=double, default=0}
 Crouch Move Head Up {type=double, default=0}
 Engage {type=double, default=0}
 HeadUp {type=double, default=0}
]
