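# Physics asset for the female WoolExpo face accessory mesh, providing capsule bodies for dynamic hair.
# Seven bone chains (three back, two front, two side) hang from the head, with limited swing and very tight twist.
# Head, neck, upper spine, clavicle and upper arm capsules form the body shapes, joined by free constraints.
# An extra free constraint cross-links the right back chain's tip to the right side chain's tip.
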PhysicsAsset /BRCosmetics/Characters/Player/Female/Medium/Bodies/F_MED_WoolExpo/Meshes/Parts/F_MED_WoolExpo_FaceAcc_Physics
bodies {
 clavicle_l [{shapeType=Capsule, length=8.5, radius=4.5, shapeName=clavicle_l_capsule, center=<6.7017 0 0>, rotation=<rot -90 -0.0002 0>}]
 clavicle_r [{shapeType=Capsule, length=8.5, radius=4.5, shapeName=clavicle_r_capsule, center=<-6.7017 0 0>, rotation=<rot -90 179.9996 0>}]
 dyn_hair_bck_c_01 [{shapeType=Capsule, length=4, radius=2, shapeName=dyn_hair_bck_c_01_capsule, center=<1.4602 0 0>, rotation=<rot 90 0 0>}]
 dyn_hair_bck_c_02 [{shapeType=Capsule, length=4, radius=2, shapeName=dyn_hair_bck_c_02_capsule, center=<1.4602 0 0>, rotation=<rot 90 0 0>}]
 dyn_hair_bck_c_03 [{shapeType=Capsule, length=4, radius=2, shapeName=dyn_hair_bck_c_03_capsule, center=<1.4602 0 0>, rotation=<rot 90 0 0>}]
 dyn_hair_bck_c_04 [{shapeType=Capsule, length=2.5, radius=2.5, shapeName=dyn_hair_bck_c_04_capsule, center=<1.4602 0 0>, rotation=<rot 90 0 0>}]
 dyn_hair_bck_l_01 [{shapeType=Capsule, length=4, radius=2, shapeName=dyn_hair_bck_l_01_capsule, center=<1.4602 0 0>, rotation=<rot 90 0 0>}]
 dyn_hair_bck_l_02 [{shapeType=Capsule, length=4, radius=2, shapeName=dyn_hair_bck_l_02_capsule, center=<1.4602 0 0>, rotation=<rot 90 0 0>}]
 dyn_hair_bck_l_03 [{shapeType=Capsule, length=4, radius=2, shapeName=dyn_hair_bck_l_03_capsule, center=<1.4602 0 0>, rotation=<rot 90 0 0>}]
 dyn_hair_bck_l_04 [{shapeType=Capsule, length=2.5, radius=2.5, shapeName=dyn_hair_bck_l_04_capsule, center=<1.4602 0 0>, rotation=<rot 90 0 0>}]
 dyn_hair_bck_r_01 [{shapeType=Capsule, length=4, radius=2, shapeName=dyn_hair_bck_r_01_capsule, center=<1.4602 0 0>, rotation=<rot 90 0 0>}]
 dyn_hair_bck_r_02 [{shapeType=Capsule, length=4, radius=2, shapeName=dyn_hair_bck_r_02_capsule, center=<1.4602 0 0>, rotation=<rot 90 0 0>}]
 dyn_hair_bck_r_03 [{shapeType=Capsule, length=4, radius=2, shapeName=dyn_hair_bck_r_03_capsule, center=<1.4602 0 0>, rotation=<rot 90 0 0>}]
 dyn_hair_bck_r_04 [{shapeType=Capsule, length=0.5, radius=3, shapeName=dyn_hair_bck_r_04_capsule, center=<1.962 -0.2496 0>, rotation=<rot 90 0 0>}]
 dyn_hair_fr_l_01 [{shapeType=Capsule, length=4, radius=2, shapeName=dyn_hair_fr_l_01_capsule, center=<1.4602 0 0>, rotation=<rot 90 0 0>}]
 dyn_hair_fr_l_02 [{shapeType=Capsule, length=4, radius=2, shapeName=dyn_hair_fr_l_02_capsule, center=<1.4602 0 0>, rotation=<rot 90 0 0>}]
 dyn_hair_fr_l_03 [{shapeType=Capsule, length=4, radius=2, shapeName=dyn_hair_fr_l_03_capsule, center=<1.4602 0 0>, rotation=<rot 90 0 0>}]
 dyn_hair_fr_r_01 [{shapeType=Capsule, length=4, radius=2, shapeName=dyn_hair_fr_r_01_capsule, center=<1.4602 0 0>, rotation=<rot 90 0 0>}]
 dyn_hair_fr_r_02 [{shapeType=Capsule, length=4, radius=2, shapeName=dyn_hair_fr_r_02_capsule, center=<1.4602 0 0>, rotation=<rot 90 0 0>}]
 dyn_hair_fr_r_03 [{shapeType=Capsule, length=4.5, radius=1.5, shapeName=dyn_hair_fr_r_03_capsule, center=<1.4602 0 0>, rotation=<rot 90 0 0>}]
 dyn_hair_side_l_01 [{shapeType=Capsule, length=4, radius=2, shapeName=dyn_hair_side_l_01_capsule, center=<1.4602 0 0>, rotation=<rot 90 0 0>}]
 dyn_hair_side_l_02 [{shapeType=Capsule, length=4, radius=2, shapeName=dyn_hair_side_l_02_capsule, center=<1.4602 0 0>, rotation=<rot 90 0 0>}]
 dyn_hair_side_l_03 [{shapeType=Capsule, length=2.5, radius=2.5, shapeName=dyn_hair_side_l_03_capsule, center=<1.4602 0 0>, rotation=<rot 90 0 0>}]
 dyn_hair_side_r_01 [{shapeType=Capsule, length=4, radius=2, shapeName=dyn_hair_side_r_01_capsule, center=<1.4602 0 0>, rotation=<rot 90 0 0>}]
 dyn_hair_side_r_02 [{shapeType=Capsule, length=4, radius=2, shapeName=dyn_hair_side_r_02_capsule, center=<1.4602 0 0>, rotation=<rot 90 0 0>}]
 dyn_hair_side_r_03 [{shapeType=Capsule, length=2.5, radius=2.5, shapeName=dyn_hair_side_r_03_capsule, center=<1.4602 0 0>, rotation=<rot 80 0 0>}]
 head [{shapeType=Capsule, length=0, radius=7.5, shapeName=head_capsule}]
 neck_02 [{shapeType=Capsule, length=4.5, radius=5.5, shapeName=neck_02_capsule, center=<-0.1473 -0.8352 0>, rotation=<rot -90 -10.0001 0>}]
 spine_05 [{shapeType=Capsule, length=9.5, radius=9.5, shapeName=spine_05_capsule, center=<1.1488 0 0>}]
 upperarm_l [{shapeType=Capsule, length=16, radius=5, shapeName=upperarm_l_capsule, center=<8.5439 0 -0.7578>, rotation=<rot 90 0.0002 0>}]
 upperarm_r [{shapeType=Capsule, length=22.5, radius=5, shapeName=upperarm_r_capsule, center=<-8.5439 0.4666 1.0672>, rotation=<rot -90 0 0>}]
}
constraints [
 {
  bone1Name=upperarm_r
  bone2Name=clavicle_r
  swing1LimitDegrees=45
  swing1Motion=Free
  swing2LimitDegrees=45
  swing2Motion=Free
  twistLimitDegrees=45
  twistMotion=Free
 }
 {
  bone1Name=upperarm_l
  bone2Name=clavicle_l
  swing1LimitDegrees=45
  swing1Motion=Free
  swing2LimitDegrees=45
  swing2Motion=Free
  twistLimitDegrees=45
  twistMotion=Free
 }
 {
  bone1Name=clavicle_l
  bone2Name=spine_05
  swing1LimitDegrees=45
  swing1Motion=Free
  swing2LimitDegrees=45
  swing2Motion=Free
  twistLimitDegrees=45
  twistMotion=Free
 }
 {
  bone1Name=clavicle_r
  bone2Name=spine_05
  swing1LimitDegrees=45
  swing1Motion=Free
  swing2LimitDegrees=45
  swing2Motion=Free
  twistLimitDegrees=45
  twistMotion=Free
 }
 {
  bone1Name=head
  bone2Name=neck_02
  swing1LimitDegrees=45
  swing1Motion=Free
  swing2LimitDegrees=45
  swing2Motion=Free
  twistLimitDegrees=45
  twistMotion=Free
 }
 {
  bone1Name=neck_02
  bone2Name=spine_05
  swing1LimitDegrees=45
  swing1Motion=Free
  swing2LimitDegrees=45
  swing2Motion=Free
  twistLimitDegrees=45
  twistMotion=Free
 }
 {bone1Name=dyn_hair_bck_l_01, bone2Name=head, swing1LimitDegrees=45, swing2LimitDegrees=45, twistLimitDegrees=1}
 {bone1Name=dyn_hair_bck_c_01, bone2Name=head, swing1LimitDegrees=45, swing2LimitDegrees=45, twistLimitDegrees=1}
 {bone1Name=dyn_hair_bck_c_02, bone2Name=dyn_hair_bck_c_01, swing1LimitDegrees=45, swing2LimitDegrees=45, twistLimitDegrees=1}
 {bone1Name=dyn_hair_bck_c_03, bone2Name=dyn_hair_bck_c_02, swing1LimitDegrees=45, swing2LimitDegrees=45, twistLimitDegrees=1}
 {bone1Name=dyn_hair_bck_c_04, bone2Name=dyn_hair_bck_c_03, swing1LimitDegrees=55, swing2LimitDegrees=55, twistLimitDegrees=1}
 {bone1Name=dyn_hair_bck_l_02, bone2Name=dyn_hair_bck_l_01, swing1LimitDegrees=45, swing2LimitDegrees=45, twistLimitDegrees=1}
 {bone1Name=dyn_hair_bck_l_03, bone2Name=dyn_hair_bck_l_02, swing1LimitDegrees=45, swing2LimitDegrees=45, twistLimitDegrees=1}
 {bone1Name=dyn_hair_bck_l_04, bone2Name=dyn_hair_bck_l_03, swing1LimitDegrees=45, swing2LimitDegrees=45, twistLimitDegrees=1}
 {bone1Name=dyn_hair_bck_r_01, bone2Name=head, swing1LimitDegrees=45, swing2LimitDegrees=45, twistLimitDegrees=1}
 {bone1Name=dyn_hair_bck_r_02, bone2Name=dyn_hair_bck_r_01, swing1LimitDegrees=45, swing2LimitDegrees=45, twistLimitDegrees=1}
 {bone1Name=dyn_hair_bck_r_03, bone2Name=dyn_hair_bck_r_02, swing1LimitDegrees=45, swing2LimitDegrees=45, twistLimitDegrees=1}
 {bone1Name=dyn_hair_bck_r_04, bone2Name=dyn_hair_bck_r_03, swing1LimitDegrees=55, swing2LimitDegrees=55, twistLimitDegrees=1}
 {bone1Name=dyn_hair_fr_l_01, bone2Name=head, swing1LimitDegrees=45, swing2LimitDegrees=45, twistLimitDegrees=1}
 {bone1Name=dyn_hair_fr_l_02, bone2Name=dyn_hair_fr_l_01, swing1LimitDegrees=45, swing2LimitDegrees=45, twistLimitDegrees=1}
 {bone1Name=dyn_hair_fr_l_03, bone2Name=dyn_hair_fr_l_02, swing1LimitDegrees=45, swing2LimitDegrees=45, twistLimitDegrees=1}
 {bone1Name=dyn_hair_fr_r_01, bone2Name=head, swing1LimitDegrees=45, swing2LimitDegrees=45, twistLimitDegrees=1}
 {bone1Name=dyn_hair_fr_r_02, bone2Name=dyn_hair_fr_r_01, swing1LimitDegrees=45, swing2LimitDegrees=45, twistLimitDegrees=1}
 {bone1Name=dyn_hair_fr_r_03, bone2Name=dyn_hair_fr_r_02, swing1LimitDegrees=45, swing2LimitDegrees=45, twistLimitDegrees=1}
 {bone1Name=dyn_hair_side_l_01, bone2Name=head, swing1LimitDegrees=45, swing2LimitDegrees=45, twistLimitDegrees=1}
 {bone1Name=dyn_hair_side_l_02, bone2Name=dyn_hair_side_l_01, swing1LimitDegrees=45, swing2LimitDegrees=45, twistLimitDegrees=1}
 {bone1Name=dyn_hair_side_l_03, bone2Name=dyn_hair_side_l_02, swing1LimitDegrees=45, swing2LimitDegrees=45, twistLimitDegrees=1}
 {bone1Name=dyn_hair_side_r_01, bone2Name=head, swing1LimitDegrees=45, swing2LimitDegrees=45, twistLimitDegrees=1}
 {bone1Name=dyn_hair_side_r_02, bone2Name=dyn_hair_side_r_01, swing1LimitDegrees=45, swing2LimitDegrees=45, twistLimitDegrees=1}
 {bone1Name=dyn_hair_side_r_03, bone2Name=dyn_hair_side_r_02, swing1LimitDegrees=55, swing2LimitDegrees=55, twistLimitDegrees=1}
 {
  bone1Name=dyn_hair_bck_r_04
  bone2Name=dyn_hair_side_r_03
  swing1LimitDegrees=45
  swing1Motion=Free
  swing2LimitDegrees=45
  swing2Motion=Free
  twistLimitDegrees=45
  twistMotion=Free
 }
]
