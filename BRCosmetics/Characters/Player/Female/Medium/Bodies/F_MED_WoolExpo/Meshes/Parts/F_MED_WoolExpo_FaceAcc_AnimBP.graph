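# Post-copy AnimBP for the F_MED_WoolExpo face accessory, adding secondary motion to its ears and ear fluff.
# The attached parent's pose feeds a Control Rig, then a neck_02-space RigidBody node.
# Chained, spring-limited AnimDynamics nodes follow: spine_04-referenced RootRelative for ears, root-referenced BoneRelative for fluff.
# On customization load, front-end state sets Simulation Space and Frontend Alpha, though the AnimGraph shows no reads of either.
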AnimBlueprint /BRCosmetics/Characters/Player/Female/Medium/Bodies/F_MED_WoolExpo/Meshes/Parts/F_MED_WoolExpo_FaceAcc_AnimBP
target_skeleton=&/F_MED_WoolExpo_FaceAcc_Skeleton
graphs [
 AnimGraph {nodes [
   n1 = Control Rig(Source=Copy Pose From Mesh() {bUseAttachedParent=true}) {
    alphaBoolBlend {blendOption=Linear}
    controlRigAssetReference {blueprintRigClass=&/F_MED_WoolExpo_FaceAcc_CtrlRig.F_MED_WoolExpo_FaceAcc_CtrlRig_C}
   }
   n2 = RigidBody(ComponentPose=Local To Component(LocalPose=@n1)) {
    bClampLinearTranslationLimitToRefPose=true
    bOverrideWorldGravity=true
    overridePhysicsAsset=&/F_MED_WoolExpo_FaceAcc_Physics
    simulationSpace=BaseBoneSpace
    alphaBoolBlend {blendOption=Linear}
    baseBoneRef {boneName=neck_02}
    overrideWorldGravity {z=-700}
    simSpaceSettings {maxAngularAcceleration=8000, maxAngularVelocity=8, maxLinearAcceleration=8000, maxLinearVelocity=800, velocityScaleZ=1.2, worldAlpha=0.3}
   }
   n3 = Anim Dynamics - Bone: dyn_ear_l_01!(SimulationSpace:AnimPhysSimSpaceType Enum=Component, ComponentPose=@n2, Alpha=1, ExternalForce="0,0,0") -> (Pose) {
    angularDampingOverride=1
    angularSpringConstant=12000
    bAngularSpring=true
    bOverrideAngularDamping=true
    bUseGravityOverride=true
    gravityScale=0.35
    simulationSpace=RootRelative
    alphaBoolBlend {blendOption=Linear}
    boundBone {boneName=dyn_ear_l_01}
    componentAppliedLinearAccClamp=<1300 1300 1300>
    componentLinearAccScale=<1 1 1>
    componentLinearVelScale=<1 1 1>
    gravityOverride {z=-980}
    physicsBodyDefinitions [{
      boundBone {boneName=dyn_ear_l_01}
      boxExtents=<55 55 55>
      constraintSetup {angularLimitsMax=<2 5 2>, angularLimitsMin=<-2 -5 -2>, angularTarget=<1 0 0>}
      localJointOffset=<1 0 0>
     }]
    relativeSpaceBone {boneName=spine_04}
   }
   n4 = Anim Dynamics - Bone: dyn_ear_l_02!(SimulationSpace:AnimPhysSimSpaceType Enum, ComponentPose=@n3, Alpha=1, ExternalForce="0,0,0") -> (Pose) {
    angularDampingOverride=1
    angularSpringConstant=12000
    bAngularSpring=true
    bOverrideAngularDamping=true
    bUseGravityOverride=true
    gravityScale=0.35
    simulationSpace=RootRelative
    alphaBoolBlend {blendOption=Linear}
    boundBone {boneName=dyn_ear_l_02}
    componentAppliedLinearAccClamp=<1300 1300 1300>
    componentLinearAccScale=<1 1 1>
    componentLinearVelScale=<1 1 1>
    gravityOverride {z=-980}
    physicsBodyDefinitions [{
      boundBone {boneName=dyn_ear_l_02}
      boxExtents=<25 25 5>
      constraintSetup {angularLimitsMax=<2 5 2>, angularLimitsMin=<-2 -5 -2>, angularTarget=<1 0 0>}
      localJointOffset=<1 0 0>
     }]
    relativeSpaceBone {boneName=spine_04}
   }
   n5 = Anim Dynamics - Bone: dyn_ear_l_03!(SimulationSpace:AnimPhysSimSpaceType Enum, ComponentPose=@n4, Alpha=1, ExternalForce="0,0,0") -> (Pose) {
    angularDampingOverride=1
    angularSpringConstant=12000
    bAngularSpring=true
    bOverrideAngularDamping=true
    bUseGravityOverride=true
    gravityScale=0.35
    simulationSpace=RootRelative
    alphaBoolBlend {blendOption=Linear}
    boundBone {boneName=dyn_ear_l_03}
    componentAppliedLinearAccClamp=<1300 1300 1300>
    componentLinearAccScale=<1 1 1>
    componentLinearVelScale=<1 1 1>
    gravityOverride {z=-980}
    physicsBodyDefinitions [{
      boundBone {boneName=dyn_ear_l_03}
      boxExtents=<25 25 5>
      constraintSetup {angularLimitsMax=<2 5 2>, angularLimitsMin=<-2 -5 -2>, angularTarget=<1 0 0>}
      localJointOffset=<1 0 0>
     }]
    relativeSpaceBone {boneName=spine_04}
   }
   n6 = Anim Dynamics - Bone: dyn_ear_r_01!(SimulationSpace:AnimPhysSimSpaceType Enum, ComponentPose=@n5, Alpha=1, ExternalForce="0,0,0") -> (Pose) {
    angularDampingOverride=1
    angularSpringConstant=10000
    bAngularSpring=true
    bOverrideAngularDamping=true
    bUseGravityOverride=true
    gravityScale=0.35
    simulationSpace=RootRelative
    alphaBoolBlend {blendOption=Linear}
    boundBone {boneName=dyn_ear_r_01}
    componentAppliedLinearAccClamp=<1300 1300 1300>
    componentLinearAccScale=<1 1 1>
    componentLinearVelScale=<1 1 1>
    gravityOverride {z=-980}
    physicsBodyDefinitions [{
      boundBone {boneName=dyn_ear_r_01}
      boxExtents=<55 55 55>
      constraintSetup {angularLimitsMax=<2 10 2>, angularLimitsMin=<-2 -10 -2>, angularTarget=<1 0 0>}
      localJointOffset=<1 0 0>
     }]
    relativeSpaceBone {boneName=spine_04}
   }
   n7 = Anim Dynamics - Bone: dyn_ear_r_02!(SimulationSpace:AnimPhysSimSpaceType Enum, ComponentPose=@n6, Alpha=1, ExternalForce="0,0,0") -> (Pose) {
    angularDampingOverride=1
    angularSpringConstant=10000
    bAngularSpring=true
    bOverrideAngularDamping=true
    bUseGravityOverride=true
    gravityScale=0.35
    simulationSpace=RootRelative
    alphaBoolBlend {blendOption=Linear}
    boundBone {boneName=dyn_ear_r_02}
    componentAppliedLinearAccClamp=<1300 1300 1300>
    componentLinearAccScale=<1 1 1>
    componentLinearVelScale=<1 1 1>
    gravityOverride {z=-980}
    physicsBodyDefinitions [{
      boundBone {boneName=dyn_ear_r_02}
      boxExtents=<20 20 20>
      constraintSetup {angularLimitsMax=<2 10 2>, angularLimitsMin=<-2 -10 -2>, angularTarget=<1 0 0>}
      localJointOffset=<1 0 0>
     }]
    relativeSpaceBone {boneName=spine_04}
   }
   n8 = Anim Dynamics - Bone: dyn_ear_r_03!(SimulationSpace:AnimPhysSimSpaceType Enum, ComponentPose=@n7, Alpha=1, ExternalForce="0,0,0") -> (Pose) {
    angularDampingOverride=1
    angularSpringConstant=10000
    bAngularSpring=true
    bOverrideAngularDamping=true
    bUseGravityOverride=true
    gravityScale=0.35
    simulationSpace=RootRelative
    alphaBoolBlend {blendOption=Linear}
    boundBone {boneName=dyn_ear_r_03}
    componentAppliedLinearAccClamp=<1300 1300 1300>
    componentLinearAccScale=<1 1 1>
    componentLinearVelScale=<1 1 1>
    gravityOverride {z=-980}
    physicsBodyDefinitions [{
      boundBone {boneName=dyn_ear_r_03}
      boxExtents=<20 20 20>
      constraintSetup {angularLimitsMax=<2 10 2>, angularLimitsMin=<-2 -10 -2>, angularTarget=<1 0 0>}
      localJointOffset=<1 0 0>
     }]
    relativeSpaceBone {boneName=spine_04}
   }
   n9 = Anim Dynamics - Bone: dyn_earFluff_l_A_01!(SimulationSpace:AnimPhysSimSpaceType Enum=BoneRelative, ComponentPose=@n8, Alpha=1, ExternalForce="0,0,0") -> (Pose) {
    angularDampingOverride=0.99
    angularSpringConstant=5000
    bAngularSpring=true
    bGravityOverrideInSimSpace=true
    bOverrideAngularDamping=true
    bUseGravityOverride=true
    simulationSpace=BoneRelative
    alphaBoolBlend {blendOption=Linear}
    boundBone {boneName=dyn_earFluff_l_A_01}
    componentAppliedLinearAccClamp=<1000 1000 1000>
    componentLinearAccScale=<1 1 1>
    componentLinearVelScale=<1 1 1>
    gravityOverride {z=-980}
    physicsBodyDefinitions [{
      boundBone {boneName=dyn_earFluff_l_A_01}
      boxExtents=<22 22 22>
      constraintSetup {angularLimitsMax=<10 10 10>, angularLimitsMin=<-10 -30 -10>, angularTarget=<1 0 0>}
      localJointOffset=<1 0 0>
     }]
    relativeSpaceBone {boneName=root}
   }
   n10 = Anim Dynamics - Bone: dyn_earFluff_l_A_02!(SimulationSpace:AnimPhysSimSpaceType Enum=BoneRelative, ComponentPose=@n9, Alpha=1, ExternalForce="0,0,0") -> (Pose) {
    angularDampingOverride=0.99
    angularSpringConstant=5000
    bAngularSpring=true
    bGravityOverrideInSimSpace=true
    bOverrideAngularDamping=true
    bUseGravityOverride=true
    simulationSpace=BoneRelative
    alphaBoolBlend {blendOption=Linear}
    boundBone {boneName=dyn_earFluff_l_A_02}
    componentAppliedLinearAccClamp=<1000 1000 1000>
    componentLinearAccScale=<1 1 1>
    componentLinearVelScale=<1 1 1>
    gravityOverride {z=-980}
    physicsBodyDefinitions [{
      boundBone {boneName=dyn_earFluff_l_A_02}
      boxExtents=<22 22 22>
      constraintSetup {angularLimitsMax=<10 10 10>, angularLimitsMin=<-10 -30 -10>, angularTarget=<1 0 0>}
      localJointOffset=<1 0 0>
     }]
    relativeSpaceBone {boneName=root}
   }
   n11 = Anim Dynamics - Bone: dyn_earFluff_l_B!(SimulationSpace:AnimPhysSimSpaceType Enum=BoneRelative, ComponentPose=@n10, Alpha=1, ExternalForce="0,0,0") -> (Pose) {
    angularDampingOverride=0.99
    angularSpringConstant=6000
    bAngularSpring=true
    bGravityOverrideInSimSpace=true
    bOverrideAngularDamping=true
    simulationSpace=BoneRelative
    alphaBoolBlend {blendOption=Linear}
    boundBone {boneName=dyn_earFluff_l_B}
    componentAppliedLinearAccClamp=<1000 1000 1000>
    componentLinearAccScale=<1 1 1>
    componentLinearVelScale=<1 1 1>
    gravityOverride {z=-980}
    physicsBodyDefinitions [{
      boundBone {boneName=dyn_earFluff_l_B}
      boxExtents=<22 22 22>
      constraintSetup {angularLimitsMax=<10 10 10>, angularLimitsMin=<-10 -30 -10>, angularTarget=<1 0 0>}
      localJointOffset=<0.6 0 0>
     }]
    relativeSpaceBone {boneName=root}
   }
   n12 = Anim Dynamics - Bone: dyn_earFluff_l_C!(SimulationSpace:AnimPhysSimSpaceType Enum=BoneRelative, ComponentPose=@n11, Alpha=1, ExternalForce="0,0,0") -> (Pose) {
    angularDampingOverride=0.99
    angularSpringConstant=6000
    bAngularSpring=true
    bGravityOverrideInSimSpace=true
    bOverrideAngularDamping=true
    simulationSpace=BoneRelative
    alphaBoolBlend {blendOption=Linear}
    boundBone {boneName=dyn_earFluff_l_C}
    componentAppliedLinearAccClamp=<1000 1000 1000>
    componentLinearAccScale=<1 1 1>
    componentLinearVelScale=<1 1 1>
    gravityOverride {z=-980}
    physicsBodyDefinitions [{
      boundBone {boneName=dyn_earFluff_l_C}
      boxExtents=<25 25 25>
      constraintSetup {angularLimitsMax=<10 10 10>, angularLimitsMin=<-10 -30 -10>, angularTarget=<1 0 0>}
      localJointOffset=<0.7 0 0>
     }]
    relativeSpaceBone {boneName=root}
   }
   n13 = Anim Dynamics - Bone: dyn_earFluff_l_D!(SimulationSpace:AnimPhysSimSpaceType Enum=BoneRelative, ComponentPose=@n12, Alpha=1, ExternalForce="0,0,0") -> (Pose) {
    angularDampingOverride=0.99
    angularSpringConstant=6000
    bAngularSpring=true
    bGravityOverrideInSimSpace=true
    bOverrideAngularDamping=true
    simulationSpace=BoneRelative
    alphaBoolBlend {blendOption=Linear}
    boundBone {boneName=dyn_earFluff_l_D}
    componentAppliedLinearAccClamp=<1000 1000 1000>
    componentLinearAccScale=<1 1 1>
    componentLinearVelScale=<1 1 1>
    gravityOverride {z=-980}
    physicsBodyDefinitions [{
      boundBone {boneName=dyn_earFluff_l_D}
      boxExtents=<24 24 24>
      constraintSetup {angularLimitsMax=<10 10 10>, angularLimitsMin=<-10 -30 -10>, angularTarget=<1 0 0>}
      localJointOffset=<0.8 0 0>
     }]
    relativeSpaceBone {boneName=root}
   }
   n14 = Anim Dynamics - Bone: dyn_earFluff_l_E!(SimulationSpace:AnimPhysSimSpaceType Enum=BoneRelative, ComponentPose=@n13, Alpha=1, ExternalForce="0,0,0") -> (Pose) {
    angularDampingOverride=0.99
    angularSpringConstant=6000
    bAngularSpring=true
    bGravityOverrideInSimSpace=true
    bOverrideAngularDamping=true
    simulationSpace=BoneRelative
    alphaBoolBlend {blendOption=Linear}
    boundBone {boneName=dyn_earFluff_l_E}
    componentAppliedLinearAccClamp=<1000 1000 1000>
    componentLinearAccScale=<1 1 1>
    componentLinearVelScale=<1 1 1>
    gravityOverride {z=-980}
    physicsBodyDefinitions [{
      boundBone {boneName=dyn_earFluff_l_E}
      boxExtents=<23 23 23>
      constraintSetup {angularLimitsMax=<10 10 10>, angularLimitsMin=<-10 -30 -10>, angularTarget=<1 0 0>}
      localJointOffset=<0.9 0 0>
     }]
    relativeSpaceBone {boneName=root}
   }
   n15 = Anim Dynamics - Bone: dyn_earFluff_r_A_01!(SimulationSpace:AnimPhysSimSpaceType Enum=BoneRelative, ComponentPose=@n14, Alpha=1, ExternalForce="0,0,0") -> (Pose) {
    angularDampingOverride=0.99
    angularSpringConstant=5000
    bAngularSpring=true
    bGravityOverrideInSimSpace=true
    bOverrideAngularDamping=true
    bUseGravityOverride=true
    simulationSpace=BoneRelative
    alphaBoolBlend {blendOption=Linear}
    boundBone {boneName=dyn_earFluff_r_A_01}
    componentAppliedLinearAccClamp=<1000 1000 1000>
    componentLinearAccScale=<1 1 1>
    componentLinearVelScale=<1 1 1>
    gravityOverride {z=-980}
    physicsBodyDefinitions [{
      boundBone {boneName=dyn_earFluff_r_A_01}
      boxExtents=<22 22 22>
      constraintSetup {angularLimitsMax=<10 10 10>, angularLimitsMin=<-10 -30 -10>, angularTarget=<1 0 0>}
      localJointOffset=<1 0 0>
     }]
    relativeSpaceBone {boneName=root}
   }
   n16 = Anim Dynamics - Bone: dyn_earFluff_r_A_02!(SimulationSpace:AnimPhysSimSpaceType Enum=BoneRelative, ComponentPose=@n15, Alpha=1, ExternalForce="0,0,0") -> (Pose) {
    angularDampingOverride=0.99
    angularSpringConstant=5000
    bAngularSpring=true
    bGravityOverrideInSimSpace=true
    bOverrideAngularDamping=true
    bUseGravityOverride=true
    simulationSpace=BoneRelative
    alphaBoolBlend {blendOption=Linear}
    boundBone {boneName=dyn_earFluff_r_A_02}
    componentAppliedLinearAccClamp=<1000 1000 1000>
    componentLinearAccScale=<1 1 1>
    componentLinearVelScale=<1 1 1>
    gravityOverride {z=-980}
    physicsBodyDefinitions [{
      boundBone {boneName=dyn_earFluff_r_A_02}
      boxExtents=<22 22 22>
      constraintSetup {angularLimitsMax=<10 10 10>, angularLimitsMin=<-10 -30 -10>, angularTarget=<1 0 0>}
      localJointOffset=<1 0 0>
     }]
    relativeSpaceBone {boneName=root}
   }
   n17 = Anim Dynamics - Bone: dyn_earFluff_r_B!(SimulationSpace:AnimPhysSimSpaceType Enum=BoneRelative, ComponentPose=@n16, Alpha=1, ExternalForce="0,0,0") -> (Pose) {
    angularDampingOverride=0.99
    angularSpringConstant=6000
    bAngularSpring=true
    bGravityOverrideInSimSpace=true
    bOverrideAngularDamping=true
    simulationSpace=BoneRelative
    alphaBoolBlend {blendOption=Linear}
    boundBone {boneName=dyn_earFluff_r_B}
    componentAppliedLinearAccClamp=<1000 1000 1000>
    componentLinearAccScale=<1 1 1>
    componentLinearVelScale=<1 1 1>
    gravityOverride {z=-980}
    physicsBodyDefinitions [{
      boundBone {boneName=dyn_earFluff_r_B}
      boxExtents=<25 25 25>
      constraintSetup {angularLimitsMax=<10 10 10>, angularLimitsMin=<-10 -30 -10>, angularTarget=<1 0 0>}
      localJointOffset=<0.6 0 0>
     }]
    relativeSpaceBone {boneName=root}
   }
   n18 = Anim Dynamics - Bone: dyn_earFluff_r_C!(SimulationSpace:AnimPhysSimSpaceType Enum=BoneRelative, ComponentPose=@n17, Alpha=1, ExternalForce="0,0,0") -> (Pose) {
    angularDampingOverride=0.99
    angularSpringConstant=6000
    bAngularSpring=true
    bGravityOverrideInSimSpace=true
    bOverrideAngularDamping=true
    simulationSpace=BoneRelative
    alphaBoolBlend {blendOption=Linear}
    boundBone {boneName=dyn_earFluff_r_C}
    componentAppliedLinearAccClamp=<1000 1000 1000>
    componentLinearAccScale=<1 1 1>
    componentLinearVelScale=<1 1 1>
    gravityOverride {z=-980}
    physicsBodyDefinitions [{
      boundBone {boneName=dyn_earFluff_r_C}
      boxExtents=<22 22 22>
      constraintSetup {angularLimitsMax=<10 10 10>, angularLimitsMin=<-10 -30 -10>, angularTarget=<1 0 0>}
      localJointOffset=<0.7 0 0>
     }]
    relativeSpaceBone {boneName=root}
   }
   n19 = Anim Dynamics - Bone: dyn_earFluff_r_D!(SimulationSpace:AnimPhysSimSpaceType Enum=BoneRelative, ComponentPose=@n18, Alpha=1, ExternalForce="0,0,0") -> (Pose) {
    angularDampingOverride=0.99
    angularSpringConstant=6000
    bAngularSpring=true
    bGravityOverrideInSimSpace=true
    bOverrideAngularDamping=true
    simulationSpace=BoneRelative
    alphaBoolBlend {blendOption=Linear}
    boundBone {boneName=dyn_earFluff_r_D}
    componentAppliedLinearAccClamp=<1000 1000 1000>
    componentLinearAccScale=<1 1 1>
    componentLinearVelScale=<1 1 1>
    gravityOverride {z=-980}
    physicsBodyDefinitions [{
      boundBone {boneName=dyn_earFluff_r_D}
      boxExtents=<24 24 24>
      constraintSetup {angularLimitsMax=<10 10 10>, angularLimitsMin=<-10 -30 -10>, angularTarget=<1 0 0>}
      localJointOffset=<0.8 0 0>
     }]
    relativeSpaceBone {boneName=root}
   }
   n20 = Anim Dynamics - Bone: dyn_earFluff_r_E!(SimulationSpace:AnimPhysSimSpaceType Enum=BoneRelative, ComponentPose=@n19, Alpha=1, ExternalForce="0,0,0") -> (Pose) {
    angularDampingOverride=0.99
    angularSpringConstant=6000
    bAngularSpring=true
    bGravityOverrideInSimSpace=true
    bOverrideAngularDamping=true
    simulationSpace=BoneRelative
    alphaBoolBlend {blendOption=Linear}
    boundBone {boneName=dyn_earFluff_r_E}
    componentAppliedLinearAccClamp=<1000 1000 1000>
    componentLinearAccScale=<1 1 1>
    componentLinearVelScale=<1 1 1>
    gravityOverride {z=-980}
    physicsBodyDefinitions [{
      boundBone {boneName=dyn_earFluff_r_E}
      boxExtents=<23 23 23>
      constraintSetup {angularLimitsMax=<10 10 10>, angularLimitsMin=<-10 -30 -10>, angularTarget=<1 0 0>}
      localJointOffset=<0.9 0 0>
     }]
    relativeSpaceBone {boneName=root}
   }
   Output Pose(Result=Component To Local(ComponentPose=@n20)) {layerGroup=None}
  ]}
 EventGraph {nodes [
   Event OnCharacterCustomizationLoaded()
   > n1 = Branch(Condition=Get bIsInFrontEnd())
   > Set Simulation Space(Simulation Space:AnimPhysSimSpaceType Enum=World) -> (Output_Get:AnimPhysSimSpaceType Enum)
   > Set Frontend Alpha(Frontend Alpha:double=0.25) -> (Output_Get:double)
   Set Simulation Space(execute=@n1.else)
   > Set Frontend Alpha(Frontend Alpha=1)
   Event BlueprintUpdateAnimation()
  ]}
]
variables [
 Frontend Alpha {type=number}
 Simulation Space {type=AnimPhysSimSpaceType, default=RootRelative}
]
